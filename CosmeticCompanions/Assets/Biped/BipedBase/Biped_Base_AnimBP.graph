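# Base Animation Blueprint for biped cosmetic companions, driving locomotion, idles, moods and expressive antics.
# A Locomotion state machine uses cached Idle SM and linked-layer run poses, then layers mood, face antic, breathing, turn and tail additives before a look-at aim offset.
# Voice-over jaw curve, slots, dead blending, slope-matched root rotation and foot/hand IK finish the pose.
# Proxy-table Choosers select stops, special idles and antic poses, keeping animation choices data-driven.
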AnimBlueprint /CosmeticCompanions/Assets/Biped/BipedBase/Biped_Base_AnimBP
target_skeleton=/CosmeticCompanions/Assets/Biped/FrankenBruin/Meshes/FrankenBruin_Biped_Skeleton
graphs [
 AnimGraph {nodes [
   n1 = Blend Poses by bool(
    BlendPose_0=Modify Curve(SourcePose=Use cached pose 'PreVOPose'(), CurveValues_0=1, Alpha=Get AudioEnvelopeValue() -> (AudioEnvelopeValue:float)) {curveNames [jaw_open_pose], curveValues [0]}
    BlendPose_1=Use cached pose 'PreVOPose'()
    bActiveValue=Get bIsVOActive() -> (bIsVOActive:bool)
   )
   n2 = Dead Blending!(Source:Pose Link Structure=Slot 'SequencerDefault'(Source=Slot 'PostVO'(Source=@n1) {slotName=PostVO}) {slotName=SequencerDefault}) -> (Pose:Pose Link Structure) {is=title AnimGraphNode_DeadBlending, defaultBlendMode=Linear}
   n3 = Leg IK!(
    ComponentPose:Component Space Pose Link Structure="Transform (Modify) Bone - Bone: root"!(
     ComponentPose=Local To Component(LocalPose=@n2)
     Alpha=Get SlopeMatchingWeight() -> (SlopeMatchingWeight:float)
     Rotation=Get SlopeMatchingRotationOffset() -> (SlopeMatchingRotationOffset:Rotator)
    ) -> (Pose) {rotationMode=BMM_Additive, rotationSpace=BCS_WorldSpace, boneToModify {boneName=root}}
    Alpha:float=Get IKAlpha() -> (IKAlpha:double)
   ) -> (Pose:Component Space Pose Link Structure) {
    is=title AnimGraphNode_LegIK
    alphaBoolBlend {blendOption=Linear}
    legsDefinition [
     {
      bEnableKneeTwistCorrection=true
      bEnableRotationLimit=false
      footBoneForwardAxis=Y
      hingeRotationAxis=None
      minRotationAngle=15
      numBonesInLimb=2
      twistOffsetCurveName=None
      fKFootBone {boneName=foot_r}
      iKFootBone {boneName=IK_Foot_R}
     }
     {
      bEnableKneeTwistCorrection=true
      bEnableRotationLimit=false
      footBoneForwardAxis=Y
      hingeRotationAxis=None
      minRotationAngle=60
      numBonesInLimb=2
      twistOffsetCurveName=None
      fKFootBone {boneName=foot_l}
      iKFootBone {boneName=IK_Foot_L}
     }
    ]
   }
   n4 = Leg IK!(ComponentPose=@n3, AlphaCurveName:Name=HandIKAlpha) -> (Pose) {
    alphaInputType=Curve
    alphaBoolBlend {blendOption=Linear}
    legsDefinition [
     {
      bEnableKneeTwistCorrection=true
      bEnableRotationLimit=false
      footBoneForwardAxis=Y
      hingeRotationAxis=None
      minRotationAngle=15
      numBonesInLimb=2
      twistOffsetCurveName=None
      fKFootBone {boneName=hand_r}
      iKFootBone {boneName=IK_Hand_R}
     }
     {
      bEnableKneeTwistCorrection=true
      bEnableRotationLimit=false
      footBoneForwardAxis=Y
      hingeRotationAxis=None
      minRotationAngle=60
      numBonesInLimb=2
      twistOffsetCurveName=None
      fKFootBone {boneName=hand_l}
      iKFootBone {boneName=IK_Hand_L}
     }
    ]
   }
   Output Pose(Result=Component To Local(ComponentPose=@n4))
   n5 = Sequence Player!(Sequence:Anim Sequence Base Object Reference) -> (Pose) {bLoopAnimation=false}
   AnimGraphNode_StateMachine_3 = Locomotion SM() {is=title AnimGraphNode_StateMachine}
   n6 = Get DisableLocoFace() -> (DisableLocoFace:float)
   AnimGraphNode_StateMachine_2 = BodyMood SM() {is=title AnimGraphNode_StateMachine, bAllowConduitEntryStates=true}
   AnimGraphNode_StateMachine_1 = FaceAntic SM() {is=title AnimGraphNode_StateMachine}
   n7 = Apply Additive(
    Base=Apply Additive(
     Base=Layered blend per bone(BasePose=@AnimGraphNode_StateMachine_3, BlendPoses_0=@n5, BlendWeights_0=@n6) {layerSetup [{branchFilters [{boneName=face_attach}]}]}
     Additive=@AnimGraphNode_StateMachine_2
    ) {alphaBoolBlend {blendOption=Linear}}
    Additive=@AnimGraphNode_StateMachine_1
   ) {alphaBoolBlend {blendOption=Linear}}
   n8 = Sequence Player!(Sequence, PlayRate:float=Get BreathingPlayrateAdjust() -> (BreathingPlayrateAdjust:double)) -> (Pose) {playRateScaleBiasClampConstants {bias=1}}
   n9 = Apply Additive(
    Base=Slot 'DefaultSlot'(Source=Apply Additive!(Base=@n7, Additive=@n8, AlphaCurveName:Name=DisableBreathing) -> (Pose) {alphaInputType=Curve, alphaBoolBlend {blendOption=Linear}, alphaScaleBiasClamp {bias=1, scale=-1}})
    Additive=Blendspace Player!(X=Get CompanionDirection() -> (CompanionDirection:float), Y=0, BlendSpace:Blend Space Object Reference) -> (Pose)
    Alpha=Get TurnAdditiveAlpha() -> (TurnAdditiveAlpha:double)
   ) {alphaBoolBlend {blendOption=Linear}}
   n10 = Sequence Player!(Sequence) -> (Pose)
   n11 = Get TailAdditiveAlpha() -> (TailAdditiveAlpha:float)
   Save cached pose 'PreVOPose'(Pose="AimOffset Player '(None)'"!(
     X:float=Get LookAtYawAdjusted() -> (LookAtYawAdjusted:double)
     Y:float=Get LookAtPitchAdjusted() -> (LookAtPitchAdjusted:double)
     BlendSpace:Blend Space Object Reference
     BasePose:Pose Link Structure=Apply Additive!(Base=@n9, Additive=@n10, Alpha:float=@n11) -> (Pose) {alphaBoolBlend {blendOption=Linear}}
     AlphaCurveName:Name=DisableAimOffset
    ) -> (Pose:Pose Link Structure) {alphaInputType=Curve, alphaBoolBlend {blendOption=Linear}, alphaScaleBiasClamp {bias=1, scale=-1}}) {cacheName=PreVOPose}
   AnimGraphNode_StateMachine_5 = Idle SM() {is=title AnimGraphNode_StateMachine, bAllowConduitEntryStates=true}
   Save cached pose 'IdlePose'(Pose=@AnimGraphNode_StateMachine_5) {cacheName=IdlePose}
   Save cached pose 'RunPose'(Pose=LocomotionRun!() -> (Pose:Pose Link Structure) {is=title AnimGraphNode_LinkedAnimLayer, layer=LocomotionRun}) {cacheName=RunPose}
   Biped Slope Warping!(ComponentPose:Component Space Pose Link Structure, Alpha:float=1, MaxStepHeight:float=50) -> (Pose:Component Space Pose Link Structure) {
    is=title FortAnimGraphNode_SlopeWarping
    feetDefinitions [
     {numBonesInLimb=2, fKFootBone {boneName=ball_r}, iKFootBone {boneName=IK_Foot_R}}
     {numBonesInLimb=2, fKFootBone {boneName=ball_l}, iKFootBone {boneName=IK_Foot_L}}
    ]
    iKFootRootBone {boneName=IK_Foot_Root}
    pelvisBone {boneName=pelvis}
   }
  ]}
 FaceAntic SM {
  entry=Default
  path=AnimGraph.AnimGraphNode_StateMachine_1.FaceAntic SM
  ids=id
  nodes [
   AnimStateNode_0 = Default() {is=title AnimStateNode}
   AnimStateNode_1 = Antic() {is=title AnimStateNode}
   AnimStateTransitionNode_0 = AnimStateTransitionNode() {crossfadeDuration=0, logicType=TLT_Inertialization}
   AnimStateTransitionNode_1 = AnimStateTransitionNode() {automaticRuleTriggerTime=0, bAutomaticRuleBasedOnSequencePlayerInState=true, crossfadeDuration=0, logicType=TLT_Inertialization}
   AnimStateTransitionNode_2 = AnimStateTransitionNode()
  ]
  links [
   {
    dst=Default
    rule=AnimGraph.AnimGraphNode_StateMachine_1.FaceAntic SM.AnimStateTransitionNode_1.Transition
    src=Antic
    via=AnimStateTransitionNode_1
   }
   {
    dst=Default
    rule=AnimGraph.AnimGraphNode_StateMachine_1.FaceAntic SM.AnimStateTransitionNode_2.Transition
    src=Antic
    via=AnimStateTransitionNode_2
   }
   {
    dst=Antic
    rule=AnimGraph.AnimGraphNode_StateMachine_1.FaceAntic SM.AnimStateTransitionNode_0.Transition
    src=Default
    via=AnimStateTransitionNode_0
   }
  ]
 }
 Default {
  path=AnimGraph.AnimGraphNode_StateMachine_1.FaceAntic SM.AnimStateNode_0.Default
  nodes [Output Animation Pose(Result=Additive Identity Pose())]
 }
 Antic {
  path=AnimGraph.AnimGraphNode_StateMachine_1.FaceAntic SM.AnimStateNode_1.Antic
  nodes [
   AnimGraphNode_ChooserPlayer_1 = Chooser Player: PRA_Companion_FacePose!() -> (Pose:Pose Link Structure) {
    is=title AnimGraphNode_ChooserPlayer
    chooser {
     _structType=/Script/ProxyTable.LookupProxy
     proxy=/CosmeticCompanions/Chooser/PRA_Companion_FacePose
     proxyTable {
      _structType=/Script/ProxyTable.ProxyTableContextProperty
      binding {contextIndex=0, displayName=Chooser Tables, isBoundToRoot=false, propertyBindingChain [ChooserTables]}
     }
    }
    chooserContextDefinition [
     {
      _structType=/Script/Chooser.ContextObjectTypeClass
      class=&/Biped_Base_AnimBP.Biped_Base_AnimBP_C
      direction=ReadWrite
     }
     {struct=/Script/Chooser.ChooserPlayerSettings, _structType=/Script/Chooser.ContextObjectTypeStruct, direction=Read}
    ]
   }
   Output Animation Pose(Result=@AnimGraphNode_ChooserPlayer_1)
  ]
 }
 AnimationBlendStackGraph_0 {
  path=AnimGraph.AnimGraphNode_StateMachine_1.FaceAntic SM.AnimStateNode_1.Antic.AnimGraphNode_ChooserPlayer_1.AnimationBlendStackGraph_0
  nodes [Output Pose!(Result:Pose Link Structure=Blend Stack Input!() -> (Pose:Pose Link Structure) {is=title AnimGraphNode_BlendStackInput}) {is=title AnimGraphNode_BlendStackResult}]
 }
 Transition {
  path=AnimGraph.AnimGraphNode_StateMachine_1.FaceAntic SM.AnimStateTransitionNode_0.Transition
  nodes [Result(bCanEnterTransition=Get bTriggerFaceAntic() -> (bTriggerFaceAntic:bool))]
 }
 Transition {path=AnimGraph.AnimGraphNode_StateMachine_1.FaceAntic SM.AnimStateTransitionNode_1.Transition, nodes [Result()]}
 Transition {
  path=AnimGraph.AnimGraphNode_StateMachine_1.FaceAntic SM.AnimStateTransitionNode_2.Transition
  nodes [Result(bCanEnterTransition=Get bDisableFaceAntics() -> (bDisableFaceAntics:bool))]
 }
 BodyMood SM {
  entry=Conduit
  path=AnimGraph.AnimGraphNode_StateMachine_2.BodyMood SM
  ids=id
  nodes [
   AnimStateConduitNode_0 = Conduit()
   AnimStateNode_0 = Default()
   AnimStateNode_1 = Worried() {is=title AnimStateNode}
   AnimStateNode_2 = Combat() {is=title AnimStateNode}
   AnimStateTransitionNode_0 = AnimStateTransitionNode() {bAllowInertializationForSelfTransitions=true}
   AnimStateTransitionNode_11 = AnimStateTransitionNode() {bAllowInertializationForSelfTransitions=true}
   AnimStateTransitionNode_12 = AnimStateTransitionNode() {bAllowInertializationForSelfTransitions=true}
   AnimStateTransitionNode_15 = AnimStateTransitionNode()
   AnimStateTransitionNode_16 = AnimStateTransitionNode() {bAllowInertializationForSelfTransitions=true}
   AnimStateTransitionNode_7 = AnimStateTransitionNode() {bAllowInertializationForSelfTransitions=true, priorityOrder=2}
   AnimStateTransitionNode_8 = AnimStateTransitionNode() {bAllowInertializationForSelfTransitions=true}
  ]
  links [
   {
    dst=Conduit
    rule=AnimGraph.AnimGraphNode_StateMachine_2.BodyMood SM.AnimStateTransitionNode_12.Transition
    src=Combat
    via=AnimStateTransitionNode_12
   }
   {
    dst=Combat
    rule=AnimGraph.AnimGraphNode_StateMachine_2.BodyMood SM.AnimStateTransitionNode_0.AnimationTransitionGraph_2
    src=Conduit
    via=AnimStateTransitionNode_0
   }
   {
    dst=Default
    rule=AnimGraph.AnimGraphNode_StateMachine_2.BodyMood SM.AnimStateTransitionNode_7.Transition
    src=Conduit
    via=AnimStateTransitionNode_7
   }
   {
    dst=Worried
    rule=AnimGraph.AnimGraphNode_StateMachine_2.BodyMood SM.AnimStateTransitionNode_8.Transition
    src=Conduit
    via=AnimStateTransitionNode_8
   }
   {
    dst=Combat
    rule=AnimGraph.AnimGraphNode_StateMachine_2.BodyMood SM.AnimStateTransitionNode_16.Transition
    src=Default
    via=AnimStateTransitionNode_16
   }
   {
    dst=Worried
    rule=AnimGraph.AnimGraphNode_StateMachine_2.BodyMood SM.AnimStateTransitionNode_15.Transition
    src=Default
    via=AnimStateTransitionNode_15
   }
   {
    dst=Conduit
    rule=AnimGraph.AnimGraphNode_StateMachine_2.BodyMood SM.AnimStateTransitionNode_11.Transition
    src=Worried
    via=AnimStateTransitionNode_11
   }
  ]
 }
 Conduit {path=AnimGraph.AnimGraphNode_StateMachine_2.BodyMood SM.AnimStateConduitNode_0.Conduit, nodes [Result(bCanEnterTransition=true)]}
 Default {
  path=AnimGraph.AnimGraphNode_StateMachine_2.BodyMood SM.AnimStateNode_0.Default
  nodes [
   n1 = Break Biped Base Struct!(Biped_Base_Struct:"Biped Base Struct Structure (by ref)"=Get Biped_Base_Struct() -> (Biped_Base_Struct:Biped Base Struct Structure)) -> (
    Locomotion_Idle_72_#1:Anim Sequence Object Reference
    Locomotion_Sprint_73_#2:Anim Sequence Object Reference
    Locomotion_Run_74_#3:Anim Sequence Object Reference
    Locomotion_Trot_99_#4:Anim Sequence Object Reference
    Locomotion_Start_145_#5:Anim Sequence Object Reference
    Locomotion_Pivot_147_#6:Anim Sequence Object Reference
    Locomotion_ADD_142_#7:Anim Sequence Object Reference
    Body_Turn_BS_100_#8:Blend Space 1D Object Reference
    AimOffset_80_#9:Blend Space Object Reference
    Breathing_102_#10:Anim Sequence Object Reference
    Worried_Pose_103_#11:Anim Sequence Object Reference
    Worried_Locomotion_Pose_104_#12:Anim Sequence Object Reference
    Bored_Pose_Additive_124_#13:Anim Sequence Object Reference
    Bored_Intro_125_#14:Anim Sequence Object Reference
    Bored_Loop_139_#15:Anim Sequence Object Reference
    Bored_Outro_126_#16:Anim Sequence Object Reference
    Angry_Locomotion_Pose_127_#17:Anim Sequence Object Reference
    Angry_Idle_Intro_128_#18:Anim Sequence Object Reference
    Angry_Idle_Loop_129_#19:Anim Sequence Object Reference
    Angry_Idle_Outro_130_#20:Anim Sequence Object Reference
    Zero_Pose_136_#21:Anim Sequence Object Reference
    Tail_Pose_133_#22:Anim Sequence Object Reference
    Happy_Locomotion_Pose_150_#23:Anim Sequence Object Reference
   ) {is=title K2Node_BreakStruct}
   Output Animation Pose(Result=Blend Poses by bool(
     BlendPose_0=Sequence Player(Sequence=@n1.Happy_Locomotion_Pose_150_#23)
     BlendPose_1=Additive Identity Pose()
     BlendTime_0=0.2
     BlendTime_1=0.2
     bActiveValue=Get bIsMoving() -> (bIsMoving:bool)
    ))
  ]
 }
 Worried {
  path=AnimGraph.AnimGraphNode_StateMachine_2.BodyMood SM.AnimStateNode_1.Worried
  nodes [
   n1 = Sequence Player() {bLoopAnimation=false}
   Output Animation Pose(Result=Two Way Blend(
     A=Sequence Player() {bLoopAnimation=false}
     B=Modify Curve(SourcePose=@n1, CurveValues_0=1, Alpha=1) {curveNames [DisableLocoFace], curveValues [0]}
     Alpha=Get LocomotionAlpha() -> (LocomotionAlpha:double)
    ))
  ]
 }
 Combat {
  path=AnimGraph.AnimGraphNode_StateMachine_2.BodyMood SM.AnimStateNode_2.Combat
  nodes [
   n1 = Sequence Player() {bLoopAnimation=false}
   Output Animation Pose(Result=Two Way Blend(
     A=Additive Identity Pose()
     B=Modify Curve(SourcePose=@n1, CurveValues_0=1, Alpha=1) {curveNames [DisableLocoFace], curveValues [0]}
     Alpha=Get LocomotionAlpha()
    ))
  ]
 }
 AnimationTransitionGraph_2 {
  path=AnimGraph.AnimGraphNode_StateMachine_2.BodyMood SM.AnimStateTransitionNode_0.AnimationTransitionGraph_2
  nodes [Result(bCanEnterTransition=Get bIsMoodCombat() -> (bIsMoodCombat:bool))]
 }
 Transition {
  path=AnimGraph.AnimGraphNode_StateMachine_2.BodyMood SM.AnimStateTransitionNode_11.Transition
  nodes [Result(bCanEnterTransition=NOT Boolean(A=Get bIsMoodWorried() -> (bIsMoodWorried:bool)))]
 }
 Transition {
  path=AnimGraph.AnimGraphNode_StateMachine_2.BodyMood SM.AnimStateTransitionNode_12.Transition
  nodes [Result(bCanEnterTransition=NOT Boolean(A=Get bIsMoodCombat()))]
 }
 Transition {
  path=AnimGraph.AnimGraphNode_StateMachine_2.BodyMood SM.AnimStateTransitionNode_15.Transition
  nodes [Result(bCanEnterTransition=Get bIsMoodWorried())]
 }
 Transition {
  path=AnimGraph.AnimGraphNode_StateMachine_2.BodyMood SM.AnimStateTransitionNode_16.Transition
  nodes [Result(bCanEnterTransition=Get bIsMoodCombat())]
 }
 Transition {path=AnimGraph.AnimGraphNode_StateMachine_2.BodyMood SM.AnimStateTransitionNode_7.Transition, nodes [Result(bCanEnterTransition=true)]}
 Transition {
  path=AnimGraph.AnimGraphNode_StateMachine_2.BodyMood SM.AnimStateTransitionNode_8.Transition
  nodes [Result(bCanEnterTransition=Get bIsMoodWorried())]
 }
 Locomotion SM {
  entry=Idle
  path=AnimGraph.AnimGraphNode_StateMachine_3.Locomotion SM
  ids=id
  nodes [
   ToPivot() {is=title AnimStateAliasNode, also {aliases [Stop,Walk]}}
   AnimStateConduitNode_0 = Conduit()
   AnimStateNode_0 = Start() {
    is=title AnimStateNode
    stateEntered {guid=#24}
    stateFullyBlended {guid=#25}
    stateLeft {guid=#26}
   }
   AnimStateNode_1 = Pivot() {is=title AnimStateNode}
   AnimStateNode_4 = Walk() {
    is=title AnimStateNode
    stateEntered {guid=#27}
    stateFullyBlended {guid=#28}
    stateLeft {guid=#29}
   }
   AnimStateNode_5 = Idle() {
    stateEntered {guid=#30}
    stateFullyBlended {guid=#31}
    stateLeft {guid=#32}
   }
   AnimStateNode_6 = Stop() {
    is=title AnimStateNode
    stateEntered {guid=#33}
    stateFullyBlended {guid=#34}
    stateLeft {guid=#35}
   }
   AnimStateTransitionNode_0 = AnimStateTransitionNode() {
    bAllowInertializationForSelfTransitions=true
    bAutomaticRuleBasedOnSequencePlayerInState=true
    transitionEnd {guid=#36}
    transitionInterrupt {guid=#37}
    transitionStart {guid=#38}
   }
   AnimStateTransitionNode_1 = AnimStateTransitionNode() {bAllowInertializationForSelfTransitions=true, crossfadeDuration=0.1}
   AnimStateTransitionNode_10 = AnimStateTransitionNode() {bAutomaticRuleBasedOnSequencePlayerInState=true, crossfadeDuration=0.3}
   AnimStateTransitionNode_11 = AnimStateTransitionNode() {crossfadeDuration=0.15}
   AnimStateTransitionNode_12 = AnimStateTransitionNode() {bDisabled=true, blendMode=CircularInOut, crossfadeDuration=0.3}
   AnimStateTransitionNode_2 = AnimStateTransitionNode() {bAllowInertializationForSelfTransitions=true}
   AnimStateTransitionNode_3 = AnimStateTransitionNode() {bAllowInertializationForSelfTransitions=true}
   AnimStateTransitionNode_4 = AnimStateTransitionNode() {bAllowInertializationForSelfTransitions=true, bAutomaticRuleBasedOnSequencePlayerInState=true, crossfadeDuration=0.1}
   AnimStateTransitionNode_6 = AnimStateTransitionNode() {crossfadeDuration=0.1, logicType=TLT_Inertialization, priorityOrder=2}
   AnimStateTransitionNode_9 = AnimStateTransitionNode()
  ]
  links [
   {
    dst=Start
    rule=AnimGraph.AnimGraphNode_StateMachine_3.Locomotion SM.AnimStateTransitionNode_1.Transition
    src=Conduit
    via=AnimStateTransitionNode_1
   }
   {
    dst=Walk
    rule=AnimGraph.AnimGraphNode_StateMachine_3.Locomotion SM.AnimStateTransitionNode_6.AnimationTransitionGraph_9
    src=Conduit
    via=AnimStateTransitionNode_6
   }
   {
    dst=Conduit
    rule=AnimGraph.AnimGraphNode_StateMachine_3.Locomotion SM.AnimStateTransitionNode_2.Transition
    src=Idle
    via=AnimStateTransitionNode_2
   }
   {
    dst=Walk
    rule=AnimGraph.AnimGraphNode_StateMachine_3.Locomotion SM.AnimStateTransitionNode_4.Transition
    src=Pivot
    via=AnimStateTransitionNode_4
   }
   {
    dst=Walk
    rule=AnimGraph.AnimGraphNode_StateMachine_3.Locomotion SM.AnimStateTransitionNode_0.Transition
    src=Start
    via=AnimStateTransitionNode_0
   }
   {
    dst=Idle
    rule=AnimGraph.AnimGraphNode_StateMachine_3.Locomotion SM.AnimStateTransitionNode_10.AnimationTransitionGraph_9
    src=Stop
    via=AnimStateTransitionNode_10
   }
   {
    dst=Idle
    rule=AnimGraph.AnimGraphNode_StateMachine_3.Locomotion SM.AnimStateTransitionNode_12.AnimationTransitionGraph_9
    src=Stop
    via=AnimStateTransitionNode_12
   }
   {
    dst=Walk
    rule=AnimGraph.AnimGraphNode_StateMachine_3.Locomotion SM.AnimStateTransitionNode_11.AnimationTransitionGraph_9
    src=Stop
    via=AnimStateTransitionNode_11
   }
   {
    dst=Pivot
    rule=AnimGraph.AnimGraphNode_StateMachine_3.Locomotion SM.AnimStateTransitionNode_3.Transition
    src=ToPivot
    via=AnimStateTransitionNode_3
   }
   {
    dst=Stop
    rule=AnimGraph.AnimGraphNode_StateMachine_3.Locomotion SM.AnimStateTransitionNode_9.AnimationTransitionGraph_9
    src=Walk
    via=AnimStateTransitionNode_9
   }
  ]
 }
 Conduit {
  path=AnimGraph.AnimGraphNode_StateMachine_3.Locomotion SM.AnimStateConduitNode_0.Conduit
  nodes [Result(bCanEnterTransition=Get bIsMoving())]
 }
 Start {
  path=AnimGraph.AnimGraphNode_StateMachine_3.Locomotion SM.AnimStateNode_0.Start
  nodes [Output Animation Pose(Result=Sequence Player() {bLoopAnimation=false})]
 }
 Pivot {
  path=AnimGraph.AnimGraphNode_StateMachine_3.Locomotion SM.AnimStateNode_1.Pivot
  nodes [Output Animation Pose(Result=Sequence Player() {bLoopAnimation=false})]
 }
 Walk {
  path=AnimGraph.AnimGraphNode_StateMachine_3.Locomotion SM.AnimStateNode_4.Walk
  nodes [Output Animation Pose(Result=Use cached pose 'RunPose'())]
 }
 Idle {
  path=AnimGraph.AnimGraphNode_StateMachine_3.Locomotion SM.AnimStateNode_5.Idle
  nodes [
   AnimGraphNode_StateMachine_1 = AnticsSM() {is=title AnimGraphNode_StateMachine, bAllowConduitEntryStates=true}
   AnimGraphNode_StateMachine_2 = AnticPoseSM() {is=title AnimGraphNode_StateMachine}
   n1 = Apply Additive!(
    Base=Apply Additive!(Base=Use cached pose 'IdlePose'(), Additive=@AnimGraphNode_StateMachine_1, Alpha:float=Get AnticsAlpha() -> (AnticsAlpha:double)) -> (Pose) {alphaBoolBlend {blendOption=Linear}}
    Additive=@AnimGraphNode_StateMachine_2
   ) -> (Pose) {alphaInputType=Bool, alphaBoolBlend {blendOption=Linear}}
   n2 = Use cached pose 'IdlePose'()
   n3 = Get bUseLowFidelity() -> (bUseLowFidelity:bool)
   Output Animation Pose(Result=Modify Curve(
     SourcePose=Blend Poses by bool(BlendPose_0=@n2, BlendPose_1=@n1, BlendTime_0=0.5, BlendTime_1=0.5, bActiveValue=@n3) {blendType=Linear, transitionType=Inertialization}
     CurveValues_0=1
     Alpha=1
    ) {curveNames [idle], curveValues [0]})
  ]
 }
 AnticsSM {
  entry=Conduit
  path=AnimGraph.AnimGraphNode_StateMachine_3.Locomotion SM.AnimStateNode_5.Idle.AnimGraphNode_StateMachine_1.AnticsSM
  ids=id
  nodes [
   AnimStateConduitNode_0 = Conduit()
   AnimStateNode_0 = DefaultState() {is=title AnimStateNode}
   AnimStateNode_1 = HeadAnticChooser() {is=title AnimStateNode}
   AnimStateTransitionNode_0 = AnimStateTransitionNode() {logicType=TLT_Inertialization, priorityOrder=100}
   AnimStateTransitionNode_1 = AnimStateTransitionNode() {logicType=TLT_Inertialization}
   AnimStateTransitionNode_2 = AnimStateTransitionNode() {bAutomaticRuleBasedOnSequencePlayerInState=true}
   AnimStateTransitionNode_5 = AnimStateTransitionNode()
  ]
  links [
   {
    dst=DefaultState
    rule=AnimGraph.AnimGraphNode_StateMachine_3.Locomotion SM.AnimStateNode_5.Idle.AnimGraphNode_StateMachine_1.AnticsSM.AnimStateTransitionNode_0.AnimationTransitionGraph_4
    src=Conduit
    via=AnimStateTransitionNode_0
   }
   {
    dst=HeadAnticChooser
    rule=AnimGraph.AnimGraphNode_StateMachine_3.Locomotion SM.AnimStateNode_5.Idle.AnimGraphNode_StateMachine_1.AnticsSM.AnimStateTransitionNode_1.AnimationTransitionGraph_4
    src=Conduit
    via=AnimStateTransitionNode_1
   }
   {
    dst=Conduit
    rule=AnimGraph.AnimGraphNode_StateMachine_3.Locomotion SM.AnimStateNode_5.Idle.AnimGraphNode_StateMachine_1.AnticsSM.AnimStateTransitionNode_5.AnimationTransitionGraph_4
    src=DefaultState
    via=AnimStateTransitionNode_5
   }
   {
    dst=Conduit
    rule=AnimGraph.AnimGraphNode_StateMachine_3.Locomotion SM.AnimStateNode_5.Idle.AnimGraphNode_StateMachine_1.AnticsSM.AnimStateTransitionNode_2.AnimationTransitionGraph_4
    src=HeadAnticChooser
    via=AnimStateTransitionNode_2
   }
  ]
 }
 Conduit {
  path=AnimGraph.AnimGraphNode_StateMachine_3.Locomotion SM.AnimStateNode_5.Idle.AnimGraphNode_StateMachine_1.AnticsSM.AnimStateConduitNode_0.Conduit
  nodes [Result(bCanEnterTransition=true)]
 }
 DefaultState {
  path=AnimGraph.AnimGraphNode_StateMachine_3.Locomotion SM.AnimStateNode_5.Idle.AnimGraphNode_StateMachine_1.AnticsSM.AnimStateNode_0.DefaultState
  nodes [
   AnimGraphNode_ChooserPlayer_0 = Chooser Player!() -> (Pose:Pose Link Structure) {
    is=title AnimGraphNode_ChooserPlayer
    chooserContextDefinition [
     {
      _structType=&ContextObjectTypeClass
      class=&Biped_Base_AnimBP_C
      direction=ReadWrite
     }
     {struct=&ChooserPlayerSettings, _structType=&ContextObjectTypeStruct, direction=Read}
    ]
   }
   Output Animation Pose(Result=Additive Identity Pose())
  ]
 }
 AnimationBlendStackGraph_0 {
  path=AnimGraph.AnimGraphNode_StateMachine_3.Locomotion SM.AnimStateNode_5.Idle.AnimGraphNode_StateMachine_1.AnticsSM.AnimStateNode_0.DefaultState.AnimGraphNode_ChooserPlayer_0.AnimationBlendStackGraph_0
  nodes [Output Pose(Result=Blend Stack Input())]
 }
 HeadAnticChooser {
  path=AnimGraph.AnimGraphNode_StateMachine_3.Locomotion SM.AnimStateNode_5.Idle.AnimGraphNode_StateMachine_1.AnticsSM.AnimStateNode_1.HeadAnticChooser
  nodes [Output Animation Pose(Result=Modify Curve(SourcePose=Sequence Player() {bLoopAnimation=false}, CurveValues_0=1, Alpha=1) {curveNames [DisableFaceAntics], curveValues [0]})]
 }
 AnimationTransitionGraph_4 {
  path=AnimGraph.AnimGraphNode_StateMachine_3.Locomotion SM.AnimStateNode_5.Idle.AnimGraphNode_StateMachine_1.AnticsSM.AnimStateTransitionNode_0.AnimationTransitionGraph_4
  nodes [Result(bCanEnterTransition=true)]
 }
 AnimationTransitionGraph_4 {
  path=AnimGraph.AnimGraphNode_StateMachine_3.Locomotion SM.AnimStateNode_5.Idle.AnimGraphNode_StateMachine_1.AnticsSM.AnimStateTransitionNode_1.AnimationTransitionGraph_4
  nodes [
   Result(bCanEnterTransition=Get!() -> (AnticSequenceInfo_AnticSequenceToPlay:Anim Sequence Object Reference, AnticSequenceInfo_bAnticSequenceReady:bool=~).AnticSequenceInfo_bAnticSequenceReady)
   Break Companion Random Number Timer!(CompanionRandomNumberTimer:"Companion Random Number Timer Structure (by ref)"=Get AnticsTimer() -> (AnticsTimer:Companion Random Number Timer Structure)) -> (ValueInt:int, ValueFloat:float=~, bJustChanged:bool=~) {is=title K2Node_BreakStruct}
  ]
 }
 AnimationTransitionGraph_4 {
  path=AnimGraph.AnimGraphNode_StateMachine_3.Locomotion SM.AnimStateNode_5.Idle.AnimGraphNode_StateMachine_1.AnticsSM.AnimStateTransitionNode_2.AnimationTransitionGraph_4
  nodes [Result()]
 }
 AnimationTransitionGraph_4 {
  path=AnimGraph.AnimGraphNode_StateMachine_3.Locomotion SM.AnimStateNode_5.Idle.AnimGraphNode_StateMachine_1.AnticsSM.AnimStateTransitionNode_5.AnimationTransitionGraph_4
  nodes [Result(bCanEnterTransition=NOT Boolean(A=Get bIsMoodWorried()))]
 }
 AnticPoseSM {
  entry=DefaultPose
  path=AnimGraph.AnimGraphNode_StateMachine_3.Locomotion SM.AnimStateNode_5.Idle.AnimGraphNode_StateMachine_2.AnticPoseSM
  ids=id
  nodes [
   AnimStateNode_0 = DefaultPose() {is=title AnimStateNode}
   AnimStateNode_1 = AdditivePose() {is=title AnimStateNode}
   AnimStateTransitionNode_0 = AnimStateTransitionNode()
   AnimStateTransitionNode_1 = AnimStateTransitionNode() {automaticRuleTriggerTime=0, bAutomaticRuleBasedOnSequencePlayerInState=true}
   AnimStateTransitionNode_2 = AnimStateTransitionNode() {bAllowInertializationForSelfTransitions=true, crossfadeDuration=0.3}
  ]
  links [
   {
    dst=DefaultPose
    rule=AnimGraph.AnimGraphNode_StateMachine_3.Locomotion SM.AnimStateNode_5.Idle.AnimGraphNode_StateMachine_2.AnticPoseSM.AnimStateTransitionNode_1.Transition
    src=AdditivePose
    via=AnimStateTransitionNode_1
   }
   {
    dst=DefaultPose
    rule=AnimGraph.AnimGraphNode_StateMachine_3.Locomotion SM.AnimStateNode_5.Idle.AnimGraphNode_StateMachine_2.AnticPoseSM.AnimStateTransitionNode_2.Transition
    src=AdditivePose
    via=AnimStateTransitionNode_2
   }
   {
    dst=AdditivePose
    rule=AnimGraph.AnimGraphNode_StateMachine_3.Locomotion SM.AnimStateNode_5.Idle.AnimGraphNode_StateMachine_2.AnticPoseSM.AnimStateTransitionNode_0.Transition
    src=DefaultPose
    via=AnimStateTransitionNode_0
   }
  ]
 }
 DefaultPose {
  path=AnimGraph.AnimGraphNode_StateMachine_3.Locomotion SM.AnimStateNode_5.Idle.AnimGraphNode_StateMachine_2.AnticPoseSM.AnimStateNode_0.DefaultPose
  nodes [Output Animation Pose(Result=Additive Identity Pose())]
 }
 AdditivePose {
  path=AnimGraph.AnimGraphNode_StateMachine_3.Locomotion SM.AnimStateNode_5.Idle.AnimGraphNode_StateMachine_2.AnticPoseSM.AnimStateNode_1.AdditivePose
  nodes [
   AnimGraphNode_ChooserPlayer_1 = Chooser Player: PRA_Companion_BodyPose!() -> (Pose:Pose Link Structure) {
    is=title AnimGraphNode_ChooserPlayer
    chooser {
     _structType=&LookupProxy
     proxy=/CosmeticCompanions/Chooser/PRA_Companion_BodyPose
     proxyTable {
      _structType=&ProxyTableContextProperty
      binding {contextIndex=0, displayName=Chooser Tables, isBoundToRoot=false, propertyBindingChain [ChooserTables]}
     }
    }
    chooserContextDefinition [
     {
      _structType=&ContextObjectTypeClass
      class=&Biped_Base_AnimBP_C
      direction=ReadWrite
     }
     {struct=&ChooserPlayerSettings, _structType=&ContextObjectTypeStruct, direction=Read}
    ]
   }
   Output Animation Pose(Result=@AnimGraphNode_ChooserPlayer_1)
  ]
 }
 AnimationBlendStackGraph_0 {
  path=AnimGraph.AnimGraphNode_StateMachine_3.Locomotion SM.AnimStateNode_5.Idle.AnimGraphNode_StateMachine_2.AnticPoseSM.AnimStateNode_1.AdditivePose.AnimGraphNode_ChooserPlayer_1.AnimationBlendStackGraph_0
  nodes [Output Pose(Result=Blend Stack Input())]
 }
 Transition {
  path=AnimGraph.AnimGraphNode_StateMachine_3.Locomotion SM.AnimStateNode_5.Idle.AnimGraphNode_StateMachine_2.AnticPoseSM.AnimStateTransitionNode_0.Transition
  nodes [Result(bCanEnterTransition=Get bTriggerBodyAntic() -> (bTriggerBodyAntic:bool))]
 }
 Transition {
  path=AnimGraph.AnimGraphNode_StateMachine_3.Locomotion SM.AnimStateNode_5.Idle.AnimGraphNode_StateMachine_2.AnticPoseSM.AnimStateTransitionNode_1.Transition
  nodes [Result()]
 }
 Transition {
  path=AnimGraph.AnimGraphNode_StateMachine_3.Locomotion SM.AnimStateNode_5.Idle.AnimGraphNode_StateMachine_2.AnticPoseSM.AnimStateTransitionNode_2.Transition
  nodes [Result(bCanEnterTransition=Get bExitBodyAntic() -> (bExitBodyAntic:bool))]
 }
 Stop {
  path=AnimGraph.AnimGraphNode_StateMachine_3.Locomotion SM.AnimStateNode_6.Stop
  nodes [
   AnimGraphNode_ChooserPlayer_1 = Chooser Player: PRA_Companion_Stops!() -> (Pose:Pose Link Structure) {
    is=title AnimGraphNode_ChooserPlayer
    chooser {
     _structType=&LookupProxy
     proxy=/CosmeticCompanions/Chooser/PRA_Companion_Stops
     proxyTable {
      _structType=&ProxyTableContextProperty
      binding {contextIndex=0, displayName=Chooser Tables, isBoundToRoot=false, propertyBindingChain [ChooserTables]}
     }
    }
    chooserContextDefinition [
     {
      _structType=&ContextObjectTypeClass
      class=&Biped_Base_AnimBP_C
      direction=ReadWrite
     }
     {struct=&ChooserPlayerSettings, _structType=&ContextObjectTypeStruct, direction=Read}
    ]
   }
   Output Animation Pose(Result=@AnimGraphNode_ChooserPlayer_1)
  ]
 }
 AnimationBlendStackGraph_0 {
  path=AnimGraph.AnimGraphNode_StateMachine_3.Locomotion SM.AnimStateNode_6.Stop.AnimGraphNode_ChooserPlayer_1.AnimationBlendStackGraph_0
  nodes [Output Pose(Result=Blend Stack Input())]
 }
 Transition {path=AnimGraph.AnimGraphNode_StateMachine_3.Locomotion SM.AnimStateTransitionNode_0.Transition, nodes [Result()]}
 Transition {
  path=AnimGraph.AnimGraphNode_StateMachine_3.Locomotion SM.AnimStateTransitionNode_1.Transition
  nodes [Result(bCanEnterTransition=Get HasStartAnim() -> (HasStartAnim:bool))]
 }
 AnimationTransitionGraph_9 {path=AnimGraph.AnimGraphNode_StateMachine_3.Locomotion SM.AnimStateTransitionNode_10.AnimationTransitionGraph_9, nodes [Result()]}
 AnimationTransitionGraph_9 {
  path=AnimGraph.AnimGraphNode_StateMachine_3.Locomotion SM.AnimStateTransitionNode_11.AnimationTransitionGraph_9
  nodes [Result(bCanEnterTransition=Get bTriggerStopToStart() -> (bTriggerStopToStart:bool))]
 }
 AnimationTransitionGraph_9 {
  path=AnimGraph.AnimGraphNode_StateMachine_3.Locomotion SM.AnimStateTransitionNode_12.AnimationTransitionGraph_9
  nodes [Result(bCanEnterTransition=Get bIsMoodCombat())]
 }
 Transition {path=AnimGraph.AnimGraphNode_StateMachine_3.Locomotion SM.AnimStateTransitionNode_2.Transition, nodes [Result(bCanEnterTransition=true)]}
 Transition {
  path=AnimGraph.AnimGraphNode_StateMachine_3.Locomotion SM.AnimStateTransitionNode_3.Transition
  nodes [Result(bCanEnterTransition=Get bTriggerPivot() -> (bTriggerPivot:bool))]
 }
 Transition {path=AnimGraph.AnimGraphNode_StateMachine_3.Locomotion SM.AnimStateTransitionNode_4.Transition, nodes [Result()]}
 AnimationTransitionGraph_9 {
  path=AnimGraph.AnimGraphNode_StateMachine_3.Locomotion SM.AnimStateTransitionNode_6.AnimationTransitionGraph_9
  nodes [Result(bCanEnterTransition=true)]
 }
 AnimationTransitionGraph_9 {
  path=AnimGraph.AnimGraphNode_StateMachine_3.Locomotion SM.AnimStateTransitionNode_9.AnimationTransitionGraph_9
  nodes [Result(bCanEnterTransition=Get bTriggerStop() -> (bTriggerStop:bool))]
 }
 Idle SM {
  entry=Conduit
  path=AnimGraph.AnimGraphNode_StateMachine_5.Idle SM
  ids=id
  nodes [
   CombatAlias() {is=title AnimStateAliasNode, also {aliases [Combat,CombatIntro]}}
   BoredAlias() {is=title AnimStateAliasNode, also {aliases [Bored,BoredIntro]}}
   ToCombatAlias() {is=title AnimStateAliasNode, also {aliases [Combat]}}
   AnimStateConduitNode_0 = Conduit()
   AnimStateNode_0 = Idle() {
    stateEntered {guid=#39}
    stateFullyBlended {guid=#40}
    stateLeft {guid=#41}
   }
   AnimStateNode_1 = ReactionLoop() {is=title AnimStateNode}
   AnimStateNode_2 = ReactionLoopExitTransition() {is=title AnimStateNode}
   AnimStateNode_4 = CombatIntro() {
    is=title AnimStateNode
    stateEntered {guid=#42}
    stateFullyBlended {guid=#43}
    stateLeft {guid=#44}
   }
   AnimStateNode_5 = Combat()
   AnimStateNode_6 = Bored() {is=title AnimStateNode}
   AnimStateNode_7 = BoredIntro() {is=title AnimStateNode}
   AnimStateNode_8 = BoredOutro() {is=title AnimStateNode}
   AnimStateNode_9 = CombatOutro() {is=title AnimStateNode}
   AnimStateTransitionNode_0 = AnimStateTransitionNode() {logicType=TLT_Inertialization, priorityOrder=3}
   AnimStateTransitionNode_1 = AnimStateTransitionNode() {logicType=TLT_Inertialization}
   AnimStateTransitionNode_10 = AnimStateTransitionNode() {bAllowInertializationForSelfTransitions=true}
   AnimStateTransitionNode_11 = AnimStateTransitionNode() {
    automaticRuleTriggerTime=0
    bAllowInertializationForSelfTransitions=true
    bAutomaticRuleBasedOnSequencePlayerInState=true
    transitionEnd {guid=#45}
    transitionInterrupt {guid=#46}
    transitionStart {guid=#47}
   }
   AnimStateTransitionNode_13 = AnimStateTransitionNode() {automaticRuleTriggerTime=0, bAutomaticRuleBasedOnSequencePlayerInState=true}
   AnimStateTransitionNode_14 = AnimStateTransitionNode()
   AnimStateTransitionNode_15 = AnimStateTransitionNode() {bAllowInertializationForSelfTransitions=true}
   AnimStateTransitionNode_2 = AnimStateTransitionNode() {crossfadeDuration=0.3, logicType=TLT_Inertialization}
   AnimStateTransitionNode_3 = AnimStateTransitionNode() {automaticRuleTriggerTime=0, bAutomaticRuleBasedOnSequencePlayerInState=true, crossfadeDuration=0.3, logicType=TLT_Inertialization}
   AnimStateTransitionNode_4 = AnimStateTransitionNode()
   AnimStateTransitionNode_5 = AnimStateTransitionNode() {logicType=TLT_Inertialization, priorityOrder=2}
   AnimStateTransitionNode_6 = AnimStateTransitionNode() {automaticRuleTriggerTime=0, bAllowInertializationForSelfTransitions=true, bAutomaticRuleBasedOnSequencePlayerInState=true}
   AnimStateTransitionNode_7 = AnimStateTransitionNode()
   AnimStateTransitionNode_8 = AnimStateTransitionNode() {bAutomaticRuleBasedOnSequencePlayerInState=true}
   AnimStateTransitionNode_9 = AnimStateTransitionNode() {bAllowInertializationForSelfTransitions=true}
  ]
  links [
   {
    dst=BoredOutro
    rule=AnimGraph.AnimGraphNode_StateMachine_5.Idle SM.AnimStateTransitionNode_14.AnimationTransitionGraph_4
    src=BoredAlias
    via=AnimStateTransitionNode_14
   }
   {
    dst=Bored
    rule=AnimGraph.AnimGraphNode_StateMachine_5.Idle SM.AnimStateTransitionNode_13.AnimationTransitionGraph_4
    src=BoredIntro
    via=AnimStateTransitionNode_13
   }
   {
    dst=Idle
    rule=AnimGraph.AnimGraphNode_StateMachine_5.Idle SM.AnimStateTransitionNode_8.Transition
    src=BoredOutro
    via=AnimStateTransitionNode_8
   }
   {
    dst=CombatOutro
    rule=AnimGraph.AnimGraphNode_StateMachine_5.Idle SM.AnimStateTransitionNode_15.Transition
    src=CombatAlias
    via=AnimStateTransitionNode_15
   }
   {
    dst=Combat
    rule=AnimGraph.AnimGraphNode_StateMachine_5.Idle SM.AnimStateTransitionNode_11.Transition
    src=CombatIntro
    via=AnimStateTransitionNode_11
   }
   {
    dst=Idle
    rule=AnimGraph.AnimGraphNode_StateMachine_5.Idle SM.AnimStateTransitionNode_6.Transition
    src=CombatOutro
    via=AnimStateTransitionNode_6
   }
   {
    dst=Idle
    rule=AnimGraph.AnimGraphNode_StateMachine_5.Idle SM.AnimStateTransitionNode_0.Transition
    src=Conduit
    via=AnimStateTransitionNode_0
   }
   {
    dst=ReactionLoop
    rule=AnimGraph.AnimGraphNode_StateMachine_5.Idle SM.AnimStateTransitionNode_1.Transition
    src=Conduit
    via=AnimStateTransitionNode_1
   }
   {
    dst=ReactionLoopExitTransition
    rule=AnimGraph.AnimGraphNode_StateMachine_5.Idle SM.AnimStateTransitionNode_5.Transition
    src=Conduit
    via=AnimStateTransitionNode_5
   }
   {
    dst=ToCombatAlias
    rule=AnimGraph.AnimGraphNode_StateMachine_5.Idle SM.AnimStateTransitionNode_9.Transition
    src=Conduit
    via=AnimStateTransitionNode_9
   }
   {
    dst=BoredIntro
    rule=AnimGraph.AnimGraphNode_StateMachine_5.Idle SM.AnimStateTransitionNode_7.Transition
    src=Idle
    via=AnimStateTransitionNode_7
   }
   {
    dst=CombatIntro
    rule=AnimGraph.AnimGraphNode_StateMachine_5.Idle SM.AnimStateTransitionNode_10.Transition
    src=Idle
    via=AnimStateTransitionNode_10
   }
   {
    dst=Conduit
    rule=AnimGraph.AnimGraphNode_StateMachine_5.Idle SM.AnimStateTransitionNode_4.Transition
    src=Idle
    via=AnimStateTransitionNode_4
   }
   {
    dst=ReactionLoopExitTransition
    rule=AnimGraph.AnimGraphNode_StateMachine_5.Idle SM.AnimStateTransitionNode_2.Transition
    src=ReactionLoop
    via=AnimStateTransitionNode_2
   }
   {
    dst=Idle
    rule=AnimGraph.AnimGraphNode_StateMachine_5.Idle SM.AnimStateTransitionNode_3.Transition
    src=ReactionLoopExitTransition
    via=AnimStateTransitionNode_3
   }
  ]
 }
 Conduit {path=AnimGraph.AnimGraphNode_StateMachine_5.Idle SM.AnimStateConduitNode_0.Conduit, nodes [Result(bCanEnterTransition=true)]}
 Idle {
  path=AnimGraph.AnimGraphNode_StateMachine_5.Idle SM.AnimStateNode_0.Idle
  nodes [
   Blend Poses by bool(
    BlendPose_0=Sequence Player 'RoseBruin_COR_Angry_Idle_Lp'() {sequence=/CosmeticCompanions/Assets/Biped/RoseBruin/Animation/Core/RoseBruin_COR_Angry_Idle_Lp}
    BlendTime_0=0.3
    BlendTime_1=0.4
    bActiveValue=Get bIsMoodCombat()
   ) {blendType=Linear}
   Output Animation Pose(Result=Sequence Player!(Sequence, PlayRate:float=Get IdlePlayrateAdjust() -> (IdlePlayrateAdjust:double)) -> (Pose) {groupName=Idle, method=SyncGroup, playRateScaleBiasClampConstants {bias=1}})
  ]
 }
 ReactionLoop {
  path=AnimGraph.AnimGraphNode_StateMachine_5.Idle SM.AnimStateNode_1.ReactionLoop
  nodes [
   AnimGraphNode_ChooserPlayer_1 = Chooser Player: PRA_Companion_SpecialIdle!() -> (Pose:Pose Link Structure) {
    is=title AnimGraphNode_ChooserPlayer
    chooser {
     _structType=&LookupProxy
     proxy=/CosmeticCompanions/Chooser/PRA_Companion_SpecialIdle
     proxyTable {
      _structType=&ProxyTableContextProperty
      binding {contextIndex=0, displayName=Chooser Tables, isBoundToRoot=false, propertyBindingChain [ChooserTables]}
     }
    }
    chooserContextDefinition [
     {
      _structType=&ContextObjectTypeClass
      class=&Biped_Base_AnimBP_C
      direction=ReadWrite
     }
     {struct=&ChooserPlayerSettings, _structType=&ContextObjectTypeStruct, direction=Read}
    ]
   }
   Output Animation Pose(Result=@AnimGraphNode_ChooserPlayer_1)
  ]
 }
 AnimationBlendStackGraph_0 {
  path=AnimGraph.AnimGraphNode_StateMachine_5.Idle SM.AnimStateNode_1.ReactionLoop.AnimGraphNode_ChooserPlayer_1.AnimationBlendStackGraph_0
  nodes [Output Pose(Result=Blend Stack Input())]
 }
 ReactionLoopExitTransition {
  path=AnimGraph.AnimGraphNode_StateMachine_5.Idle SM.AnimStateNode_2.ReactionLoopExitTransition
  nodes [
   AnimGraphNode_ChooserPlayer_1 = Chooser Player: PRA_Companion_ReactToIdle!() -> (Pose:Pose Link Structure) {
    is=title AnimGraphNode_ChooserPlayer
    chooser {
     _structType=&LookupProxy
     proxy=/CosmeticCompanions/Chooser/PRA_Companion_ReactToIdle
     proxyTable {
      _structType=&ProxyTableContextProperty
      binding {contextIndex=0, displayName=Chooser Tables, isBoundToRoot=false, propertyBindingChain [ChooserTables]}
     }
    }
    chooserContextDefinition [
     {
      _structType=&ContextObjectTypeClass
      class=&Biped_Base_AnimBP_C
      direction=ReadWrite
     }
     {struct=&ChooserPlayerSettings, _structType=&ContextObjectTypeStruct, direction=Read}
    ]
   }
   Output Animation Pose(Result=@AnimGraphNode_ChooserPlayer_1)
  ]
 }
 AnimationBlendStackGraph_0 {
  path=AnimGraph.AnimGraphNode_StateMachine_5.Idle SM.AnimStateNode_2.ReactionLoopExitTransition.AnimGraphNode_ChooserPlayer_1.AnimationBlendStackGraph_0
  nodes [Output Pose(Result=Blend Stack Input())]
 }
 CombatIntro {
  path=AnimGraph.AnimGraphNode_StateMachine_5.Idle SM.AnimStateNode_4.CombatIntro
  nodes [Output Animation Pose(Result=Sequence Player!(Sequence) -> (Pose) {bLoopAnimation=false})]
 }
 Combat {path=AnimGraph.AnimGraphNode_StateMachine_5.Idle SM.AnimStateNode_5.Combat, nodes [Output Animation Pose(Result=Sequence Player())]}
 Bored {
  path=AnimGraph.AnimGraphNode_StateMachine_5.Idle SM.AnimStateNode_6.Bored
  nodes [
   n1 = Sequence Player()
   n2 = Sequence Player() {bLoopAnimation=false}
   Output Animation Pose(Result=Blend Poses by bool(BlendPose_0=Sequence Player(), BlendPose_1=Apply Additive(Base=@n1, Additive=@n2), bActiveValue=Get UseBoredLoop() -> (UseBoredLoop:bool)))
  ]
 }
 BoredIntro {
  path=AnimGraph.AnimGraphNode_StateMachine_5.Idle SM.AnimStateNode_7.BoredIntro
  nodes [Output Animation Pose(Result=Sequence Player() {bLoopAnimation=false})]
 }
 BoredOutro {
  path=AnimGraph.AnimGraphNode_StateMachine_5.Idle SM.AnimStateNode_8.BoredOutro
  nodes [Output Animation Pose(Result=Sequence Player() {bLoopAnimation=false})]
 }
 CombatOutro {
  path=AnimGraph.AnimGraphNode_StateMachine_5.Idle SM.AnimStateNode_9.CombatOutro
  nodes [Output Animation Pose(Result=Sequence Player() {bLoopAnimation=false})]
 }
 Transition {path=AnimGraph.AnimGraphNode_StateMachine_5.Idle SM.AnimStateTransitionNode_0.Transition, nodes [Result(bCanEnterTransition=true)]}
 Transition {
  path=AnimGraph.AnimGraphNode_StateMachine_5.Idle SM.AnimStateTransitionNode_1.Transition
  nodes [Result(bCanEnterTransition=Get bIsEmoteActive() -> (bIsEmoteActive:bool))]
 }
 Transition {
  path=AnimGraph.AnimGraphNode_StateMachine_5.Idle SM.AnimStateTransitionNode_10.Transition
  nodes [
   Result(bCanEnterTransition=Get bIsMoodCombat())
   Get bIsMoodCombat()
  ]
 }
 Transition {path=AnimGraph.AnimGraphNode_StateMachine_5.Idle SM.AnimStateTransitionNode_11.Transition, nodes [Result()]}
 AnimationTransitionGraph_4 {path=AnimGraph.AnimGraphNode_StateMachine_5.Idle SM.AnimStateTransitionNode_13.AnimationTransitionGraph_4, nodes [Result()]}
 AnimationTransitionGraph_4 {
  path=AnimGraph.AnimGraphNode_StateMachine_5.Idle SM.AnimStateTransitionNode_14.AnimationTransitionGraph_4
  nodes [Result(bCanEnterTransition=NOT Boolean(A=Get bIsMoodBored() -> (bIsMoodBored:bool)))]
 }
 Transition {
  path=AnimGraph.AnimGraphNode_StateMachine_5.Idle SM.AnimStateTransitionNode_15.Transition
  nodes [Result(bCanEnterTransition=NOT Boolean(A=Get bIsMoodCombat()))]
 }
 Transition {
  path=AnimGraph.AnimGraphNode_StateMachine_5.Idle SM.AnimStateTransitionNode_2.Transition
  nodes [
   n1 = Get bIsPlayingEmote()
   Result(bCanEnterTransition=NOT Boolean(A=@n1))
   OR Boolean(A=@n1, B=Get bIsMoodCombat())
  ]
 }
 Transition {path=AnimGraph.AnimGraphNode_StateMachine_5.Idle SM.AnimStateTransitionNode_3.Transition, nodes [Result()]}
 Transition {path=AnimGraph.AnimGraphNode_StateMachine_5.Idle SM.AnimStateTransitionNode_4.Transition, nodes [Result(bCanEnterTransition=true)]}
 Transition {
  path=AnimGraph.AnimGraphNode_StateMachine_5.Idle SM.AnimStateTransitionNode_5.Transition
  nodes [Result(bCanEnterTransition=Get bIsReactionActive() -> (bIsReactionActive:bool))]
 }
 Transition {path=AnimGraph.AnimGraphNode_StateMachine_5.Idle SM.AnimStateTransitionNode_6.Transition, nodes [Result()]}
 Transition {
  path=AnimGraph.AnimGraphNode_StateMachine_5.Idle SM.AnimStateTransitionNode_7.Transition
  nodes [Result(bCanEnterTransition=Get bIsMoodBored())]
 }
 Transition {path=AnimGraph.AnimGraphNode_StateMachine_5.Idle SM.AnimStateTransitionNode_8.Transition, nodes [Result()]}
 Transition {
  path=AnimGraph.AnimGraphNode_StateMachine_5.Idle SM.AnimStateTransitionNode_9.Transition
  nodes [Result(bCanEnterTransition=Get bIsMoodCombat())]
 }
 EventGraph {nodes [
   Event BlueprintInitializeAnimation()
   n1 = Get Data!(self:Fort User Defined Data Asset Object Reference=GetAnimDataAsset!(self:Fort Player Companion Anim Instance Object Reference) -> (ReturnValue:Fort User Defined Data Asset Object Reference)) -> (bSucceed:bool, Dest:Biped Base Struct Structure) {is=title K2Node_CallFunction}
   > n2 = Branch(Condition=@n1)
   > n3 = Set Biped_Base_Struct(Biped_Base_Struct:Biped Base Struct Structure=@n1.Dest) -> (Output_Get:Biped Base Struct Structure)
   n4 = Get Biped_Base_Struct()
   Set bHasTailAdditive(
    execute=[@n3.then, @n7.then]
    bHasTailAdditive:bool=IsValid(Object=Break Biped Base Struct(Biped_Base_Struct=@n4).Tail_Pose_133_#22)
   ) -> (Output_Get:bool)
   n5 = Break Biped Base Struct(Biped_Base_Struct=Get Biped_Base_Struct())
   > Set UseBoredLoop(UseBoredLoop:bool=IsValid(Object=@n5.Bored_Loop_139_#15)) -> (Output_Get:bool)
   > Set HasStartAnim(HasStartAnim:bool=IsValid(Object=@n5.Locomotion_Start_145_#5)) -> (Output_Get:bool)
   > Set bHasPivotAnim(bHasPivotAnim:bool=IsValid(Object=@n5.Locomotion_Pivot_147_#6)) -> (Output_Get:bool)
   n6 = Get Data(self=Get Biped_DA() -> (Biped_DA:Fort User Defined Data Asset Object Reference))
   Branch(execute=@n2.else, Condition=@n6)
   > n7 = Set Biped_Base_Struct(Biped_Base_Struct=@n6.Dest)
  ]}
 LocomotionRun {nodes [
   n1 = Get LocomotionPlayRate() -> (LocomotionPlayRate:float)
   n2 = Blend Poses by bool(
    BlendPose_0=Sequence Player!(Sequence, PlayRate:float=@n1) -> (Pose) {groupName=Locomotion, method=SyncGroup}
    BlendPose_1=Blend Poses by bool(
     BlendPose_0=Sequence Player(PlayRate=@n1) {groupName=Locomotion, method=SyncGroup}
     BlendPose_1=Sequence Player(PlayRate=@n1) {groupName=Locomotion, method=SyncGroup}
     BlendTime_0=0.2
     BlendTime_1=0.2
     bActiveValue=Get bIsLocoRun() -> (bIsLocoRun:bool)
    ) {blendType=Sinusoidal}
    BlendTime_0=0.25
    BlendTime_1=0.25
    bActiveValue=Get bIsLocoSprint() -> (bIsLocoSprint:bool)
   ) {blendType=Sinusoidal}
   n3 = Sequence Player!(Sequence) -> (Pose)
   n4 = Get LocoAdditiveAlpha() -> (LocoAdditiveAlpha:float)
   Output Pose(Result=Modify Curve(SourcePose=Apply Additive(Base=@n2, Additive=@n3, Alpha=@n4), CurveValues_0=1, Alpha=1) {curveNames [locomotionalpha], curveValues [0]}) {is=title AnimGraphNode_Root}
  ]}
]
variables [
 Biped_Base_Struct {
  type=Biped_Base_Struct
  default {
   aimOffset=null
   angry_Idle_Intro=null
   angry_Idle_Loop=null
   angry_Idle_Outro=null
   angry_Locomotion_Pose=null
   body_Turn_BS=null
   bored_Intro=null
   bored_Loop=null
   bored_Outro=null
   bored_Pose_Additive=null
   breathing=null
   happy_Locomotion_Pose=null
   locomotion_ADD=null
   locomotion_Idle=null
   locomotion_Pivot=null
   locomotion_Run=null
   locomotion_Sprint=null
   locomotion_Start=null
   locomotion_Trot=null
   tail_Pose=null
   worried_Locomotion_Pose=null
   worried_Pose=null
   zero_Pose=null
  }
 }
 Biped_DA {type=FortUserDefinedDataAsset, default=null}
 BreathingPlayrateAdjust {type=number}
 ChooserTables {type=ProxyTable, default=null}
 HasStartAnim {type=boolean}
 IdlePlayrateAdjust {type=number}
 IKAlpha {type=number}
 RandomizationContext {type=ChooserRandomizationContext, default {}}
 TurnAdditiveAlpha {type=number}
 UseBoredLoop {type=boolean}
]
