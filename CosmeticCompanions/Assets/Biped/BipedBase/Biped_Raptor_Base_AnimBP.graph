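# Base Animation Blueprint for raptor-like biped cosmetic companions, covering idle, sitting, locomotion, moods, emotes and antics.
# Clips come from a Biped_Raptor_Base_Struct variable and proxy-table Choosers.
# An inertialized switch selects a low-fidelity state machine or a high-fidelity one layered with start, mood, body-pose and tail additives.
# Body bend, head antics, look-at aim offset, slots and dead blending follow, then slope warping and leg IK ground the feet.
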AnimBlueprint /CosmeticCompanions/Assets/Biped/BipedBase/Biped_Raptor_Base_AnimBP
target_skeleton=/CosmeticCompanions/Assets/Biped/CrescentNail/Meshes/CrescentNail_Biped_Skeleton
graphs [
 AllowsCinematicOverride {nodes [
   AllowsCinematicOverride() {is=title K2Node_FunctionEntry}
   > Return Node!(execute, ReturnValue:bool=false)
  ]}
 AnimGraph {nodes [
   n1 = Use cached pose 'PreSlopeWarping'()
   n2 = Biped Slope Warping!(
    ComponentPose:Component Space Pose Link Structure="Transform (Modify) Bone - Bone: root"!(
     ComponentPose=Local To Component(LocalPose=@n1)
     Alpha=Get SlopeMatchingWeight() -> (SlopeMatchingWeight:float)
     Rotation=Get SlopeMatchingRotationOffset() -> (SlopeMatchingRotationOffset:Rotator)
    ) -> (Pose) {rotationMode=BMM_Additive, rotationSpace=BCS_WorldSpace, boneToModify {boneName=root}}
    Alpha:float=Get SlopeMatchingWeight()
   ) -> (Pose:Component Space Pose Link Structure) {
    is=title FortAnimGraphNode_SlopeWarping
    alphaBoolBlend {blendOption=Linear}
    alphaScaleBiasClamp {bias=1, scale=-1}
    feetDefinitions [
     {numBonesInLimb=3, fKFootBone {boneName=ball_l}, iKFootBone {boneName=IK_Foot_Rear_L}}
     {numBonesInLimb=3, fKFootBone {boneName=ball_r}, iKFootBone {boneName=IK_Foot_Rear_R}}
    ]
    iKFootRootBone {boneName=IK_Foot_Front_Root}
    pelvisBone {boneName=root}
   }
   n3 = Leg IK!(ComponentPose:Component Space Pose Link Structure=@n2, Alpha:float=1) -> (Pose:Component Space Pose Link Structure) {
    is=title AnimGraphNode_LegIK
    alphaBoolBlend {blendOption=Linear}
    legsDefinition [
     {
      bEnableKneeTwistCorrection=true
      bEnableRotationLimit=false
      footBoneForwardAxis=Y
      hingeRotationAxis=None
      minRotationAngle=15
      numBonesInLimb=3
      twistOffsetCurveName=None
      fKFootBone {boneName=ball_l}
      iKFootBone {boneName=IK_Foot_Rear_L}
     }
     {
      bEnableKneeTwistCorrection=true
      bEnableRotationLimit=false
      footBoneForwardAxis=Y
      hingeRotationAxis=None
      minRotationAngle=15
      numBonesInLimb=3
      twistOffsetCurveName=None
      fKFootBone {boneName=ball_r}
      iKFootBone {boneName=IK_Foot_Rear_R}
     }
    ]
   }
   Output Pose(Result=Component To Local(ComponentPose=@n3))
   AnimGraphNode_StateMachine_6 = Head Antic() {is=title AnimGraphNode_StateMachine, bAllowConduitEntryStates=true}
   Save cached pose 'Head Antic Pose'(Pose=@AnimGraphNode_StateMachine_6) {cacheName=Head Antic Pose}
   AnimGraphNode_StateMachine_2 = Locomotion() {is=title AnimGraphNode_StateMachine}
   Save cached pose 'LocomotionPose'(Pose=@AnimGraphNode_StateMachine_2) {cacheName=LocomotionPose}
   AnimGraphNode_StateMachine_5 = LocomotionStartAdditive() {is=title AnimGraphNode_StateMachine}
   AnimGraphNode_StateMachine_7 = LocomotionStateMachine() {is=title AnimGraphNode_StateMachine, bAllowConduitEntryStates=true}
   AnimGraphNode_StateMachine_3 = BodyMoodAdditive() {is=title AnimGraphNode_StateMachine, bAllowConduitEntryStates=true}
   n4 = Apply Additive(
    Base=Apply Additive(Base=@AnimGraphNode_StateMachine_7, Additive=@AnimGraphNode_StateMachine_5) {alphaBoolBlend {blendOption=Linear}}
    Additive=@AnimGraphNode_StateMachine_3
   ) {alphaBoolBlend {blendOption=Linear}}
   AnimGraphNode_StateMachine_1 = Body Pose Additive() {is=title AnimGraphNode_StateMachine, bAllowConduitEntryStates=true}
   n5 = Get AnticAlpha() -> (AnticAlpha:double)
   n6 = Break Biped Raptor Base Struct!(Biped_Raptor_Base_Struct:"Biped Raptor Base Struct Structure (by ref)"=Get Biped_Raptor_Base_Struct() -> (Biped_Raptor_Base_Struct:Biped Raptor Base Struct Structure)) -> (
    Locomotion_Start_98_#1:Anim Sequence Object Reference
    Locomotion_Idle_72_#2:Anim Sequence Object Reference
    Locomotion_Sprint_73_#3:Anim Sequence Object Reference
    Locomotion_Run_74_#4:Anim Sequence Object Reference
    Locomotion_Jog_75_#5:Anim Sequence Object Reference
    Locomotion_Walk_76_#6:Anim Sequence Object Reference
    BodyBend_BS_68_#7:Blend Space 1D Object Reference
    BodyPose_Worried_77_#8:Anim Sequence Object Reference
    AimOffset_80_#9:Blend Space Object Reference
    StandToSit_89_#10:Anim Sequence Object Reference
    Sitting_88_#11:Anim Sequence Object Reference
    SitToStand_87_#12:Anim Sequence Object Reference
    TailBend_BS_92_#13:Blend Space 1D Object Reference
    FeatherSpreadPose_96_#14:Anim Sequence Object Reference
    Angry_Idle_Intro_101_#15:Anim Sequence Object Reference
    Angry_Idle_Loop_103_#16:Anim Sequence Object Reference
    Angry_Idle_Outro_105_#17:Anim Sequence Object Reference
    BodyPose_Angry_108_#18:Anim Sequence Object Reference
   ) {is=title K2Node_BreakStruct}
   n7 = Get CompanionDirection() -> (CompanionDirection:float)
   Save cached pose 'HighFidelityPose'(Pose=Apply Additive(
     Base=Apply Additive(Base=@n4, Additive=@AnimGraphNode_StateMachine_1, Alpha=@n5) {alphaBoolBlend {blendOption=Linear}}
     Additive=Blendspace Player!(X=@n7, Y=0, BlendSpace:Blend Space Object Reference=@n6.TailBend_BS_92_#13) -> (Pose)
    ) {alphaBoolBlend {blendOption=Linear}}) {cacheName=HighFidelityPose}
   AnimGraphNode_StateMachine_4 = LowFidelityStateMachine() {is=title AnimGraphNode_StateMachine, bAllowConduitEntryStates=true}
   Save cached pose 'LowFidelityPose'(Pose=@AnimGraphNode_StateMachine_4) {cacheName=LowFidelityPose}
   AnimGraphNode_StateMachine_0 = Idle() {is=title AnimGraphNode_StateMachine, bAllowConduitEntryStates=true}
   Save cached pose 'IdlePose'(Pose=@AnimGraphNode_StateMachine_0) {cacheName=IdlePose}
   n8 = Get Biped_Raptor_Base_Struct()
   n9 = Apply Additive(
    Base=Blend Poses by bool(
     BlendPose_0=Use cached pose 'LowFidelityPose'()
     BlendPose_1=Use cached pose 'HighFidelityPose'()
     BlendTime_0=0.2
     BlendTime_1=0.2
     bActiveValue=Get bActiveValue() -> (bActiveValue:bool)
    ) {blendType=Linear, transitionType=Inertialization}
    Additive=Blendspace Player(
     X=Get CompanionDirection()
     Y=0
     BlendSpace=Break Biped Raptor Base Struct(Biped_Raptor_Base_Struct=@n8).BodyBend_BS_68_#7
    )
    Alpha=Get BodyBendAlpha() -> (BodyBendAlpha:double)
   ) {alphaBoolBlend {blendOption=Linear}}
   n10 = Use cached pose 'Head Antic Pose'()
   n11 = Get AnticAlpha()
   n12 = "AimOffset Player '(None)'"!(
    X:float=Get LookAtYawAdjusted() -> (LookAtYawAdjusted:double)
    Y:float=Get LookAtPitchAdjusted() -> (LookAtPitchAdjusted:double)
    BlendSpace:Blend Space Object Reference=Break Biped Raptor Base Struct(Biped_Raptor_Base_Struct=Get Biped_Raptor_Base_Struct()).AimOffset_80_#9
    BasePose:Pose Link Structure=Slot 'DefaultSlot'(Source=Apply Additive(Base=@n9, Additive=@n10, Alpha=@n11) {alphaBoolBlend {blendOption=Linear}})
    AlphaCurveName:Name=DisableAimOffset
   ) -> (Pose:Pose Link Structure) {alphaInputType=Curve, alphaBoolBlend {blendOption=Linear}, alphaScaleBiasClamp {bias=1, scale=-1}}
   Save cached pose 'PreSlopeWarping'(Pose=Dead Blending!(Source:Pose Link Structure=Slot 'SequencerDefault'(Source=@n12) {slotName=SequencerDefault}) -> (Pose:Pose Link Structure) {is=title AnimGraphNode_DeadBlending, defaultBlendMode=Linear}) {cacheName=PreSlopeWarping}
  ]}
 Idle {
  entry=Conduit
  path=AnimGraph.AnimGraphNode_StateMachine_0.Idle
  ids=id
  nodes [
   Alias() {is=title AnimStateAliasNode, also {aliases [SitPose,StandToSit]}}
   AnimStateConduitNode_1 = Conduit()
   AnimStateNode_0 = StandToSit() {is=title AnimStateNode}
   AnimStateNode_1 = AngryIdleIntro() {is=title AnimStateNode}
   AnimStateNode_2 = AngryIdle() {is=title AnimStateNode}
   AnimStateNode_3 = AngryIdleExit() {
    is=title AnimStateNode
    stateEntered {guid=#19}
    stateFullyBlended {guid=#20}
    stateLeft {guid=#21}
   }
   AnimStateNode_4 = SitToStand() {is=title AnimStateNode}
   AnimStateNode_5 = SitPose() {is=title AnimStateNode}
   AnimStateNode_6 = StandingPose() {is=title AnimStateNode}
   AnimStateTransitionNode_0 = AnimStateTransitionNode() {bAutomaticRuleBasedOnSequencePlayerInState=true, logicType=TLT_Inertialization}
   AnimStateTransitionNode_1 = AnimStateTransitionNode() {logicType=TLT_Inertialization, priorityOrder=2}
   AnimStateTransitionNode_10 = AnimStateTransitionNode() {logicType=TLT_Inertialization}
   AnimStateTransitionNode_11 = AnimStateTransitionNode() {bAllowInertializationForSelfTransitions=true}
   AnimStateTransitionNode_2 = AnimStateTransitionNode() {bAllowInertializationForSelfTransitions=true, crossfadeDuration=0.1, logicType=TLT_Inertialization}
   AnimStateTransitionNode_3 = AnimStateTransitionNode() {
    automaticRuleTriggerTime=0
    bAllowInertializationForSelfTransitions=true
    bAutomaticRuleBasedOnSequencePlayerInState=true
    logicType=TLT_Inertialization
   }
   AnimStateTransitionNode_4 = AnimStateTransitionNode() {
    bAllowInertializationForSelfTransitions=true
    crossfadeDuration=0.1
    logicType=TLT_Inertialization
    transitionEnd {guid=#22}
    transitionInterrupt {guid=#23}
    transitionStart {guid=#24}
   }
   AnimStateTransitionNode_5 = AnimStateTransitionNode() {automaticRuleTriggerTime=0, bAllowInertializationForSelfTransitions=true, bAutomaticRuleBasedOnSequencePlayerInState=true}
   AnimStateTransitionNode_6 = AnimStateTransitionNode() {bAutomaticRuleBasedOnSequencePlayerInState=true, crossfadeDuration=0.3, logicType=TLT_Inertialization}
   AnimStateTransitionNode_7 = AnimStateTransitionNode()
   AnimStateTransitionNode_8 = AnimStateTransitionNode() {logicType=TLT_Inertialization}
   AnimStateTransitionNode_9 = AnimStateTransitionNode() {logicType=TLT_Inertialization, priorityOrder=100}
  ]
  links [
   {
    dst=SitToStand
    rule=AnimGraph.AnimGraphNode_StateMachine_0.Idle.AnimStateTransitionNode_8.AnimationTransitionGraph_0
    src=Alias
    via=AnimStateTransitionNode_8
   }
   {
    dst=AngryIdleExit
    rule=AnimGraph.AnimGraphNode_StateMachine_0.Idle.AnimStateTransitionNode_4.AnimationTransitionGraph_3
    src=AngryIdle
    via=AnimStateTransitionNode_4
   }
   {
    dst=StandingPose
    rule=AnimGraph.AnimGraphNode_StateMachine_0.Idle.AnimStateTransitionNode_5.AnimationTransitionGraph_3
    src=AngryIdleExit
    via=AnimStateTransitionNode_5
   }
   {
    dst=AngryIdle
    rule=AnimGraph.AnimGraphNode_StateMachine_0.Idle.AnimStateTransitionNode_3.AnimationTransitionGraph_3
    src=AngryIdleIntro
    via=AnimStateTransitionNode_3
   }
   {
    dst=AngryIdleExit
    rule=AnimGraph.AnimGraphNode_StateMachine_0.Idle.AnimStateTransitionNode_11.Transition
    src=AngryIdleIntro
    via=AnimStateTransitionNode_11
   }
   {
    dst=SitPose
    rule=AnimGraph.AnimGraphNode_StateMachine_0.Idle.AnimStateTransitionNode_10.AnimationTransitionGraph_0
    src=Conduit
    via=AnimStateTransitionNode_10
   }
   {
    dst=StandToSit
    rule=AnimGraph.AnimGraphNode_StateMachine_0.Idle.AnimStateTransitionNode_1.AnimationTransitionGraph_0
    src=Conduit
    via=AnimStateTransitionNode_1
   }
   {
    dst=StandingPose
    rule=AnimGraph.AnimGraphNode_StateMachine_0.Idle.AnimStateTransitionNode_9.AnimationTransitionGraph_0
    src=Conduit
    via=AnimStateTransitionNode_9
   }
   {
    dst=StandingPose
    rule=AnimGraph.AnimGraphNode_StateMachine_0.Idle.AnimStateTransitionNode_6.AnimationTransitionGraph_0
    src=SitToStand
    via=AnimStateTransitionNode_6
   }
   {
    dst=SitPose
    rule=AnimGraph.AnimGraphNode_StateMachine_0.Idle.AnimStateTransitionNode_0.AnimationTransitionGraph_0
    src=StandToSit
    via=AnimStateTransitionNode_0
   }
   {
    dst=AngryIdleIntro
    rule=AnimGraph.AnimGraphNode_StateMachine_0.Idle.AnimStateTransitionNode_2.AnimationTransitionGraph_3
    src=StandingPose
    via=AnimStateTransitionNode_2
   }
   {
    dst=Conduit
    rule=AnimGraph.AnimGraphNode_StateMachine_0.Idle.AnimStateTransitionNode_7.AnimationTransitionGraph_0
    src=StandingPose
    via=AnimStateTransitionNode_7
   }
  ]
 }
 Conduit {path=AnimGraph.AnimGraphNode_StateMachine_0.Idle.AnimStateConduitNode_1.Conduit, nodes [Result(bCanEnterTransition=true)]}
 StandToSit {
  path=AnimGraph.AnimGraphNode_StateMachine_0.Idle.AnimStateNode_0.StandToSit
  nodes [
   n1 = Break Biped Raptor Base Struct(Biped_Raptor_Base_Struct=Get Biped_Raptor_Base_Struct())
   Output Animation Pose(Result=Modify Curve(
     SourcePose=Sequence Player!(Sequence:Anim Sequence Base Object Reference=@n1.StandToSit_89_#10) -> (Pose)
     CurveValues_0=0
     Alpha=1
    ) {curveNames [AlphaCurve], curveValues [0]}) {
    node {stateEntryFunction {functionName=OnStateEntry}}
    stateEntryFunction {bSelfContext=true, memberGuid=#25, memberName=ToSitting_Entry}
   }
  ]
 }
 AngryIdleIntro {
  path=AnimGraph.AnimGraphNode_StateMachine_0.Idle.AnimStateNode_1.AngryIdleIntro
  nodes [Output Animation Pose(Result=Sequence Player() {bLoopAnimation=false})]
 }
 AngryIdle {path=AnimGraph.AnimGraphNode_StateMachine_0.Idle.AnimStateNode_2.AngryIdle, nodes [Output Animation Pose(Result=Sequence Player())]}
 AngryIdleExit {
  path=AnimGraph.AnimGraphNode_StateMachine_0.Idle.AnimStateNode_3.AngryIdleExit
  nodes [Output Animation Pose(Result=Sequence Player() {bLoopAnimation=false})]
 }
 SitToStand {
  path=AnimGraph.AnimGraphNode_StateMachine_0.Idle.AnimStateNode_4.SitToStand
  nodes [
   n1 = Break Biped Raptor Base Struct(Biped_Raptor_Base_Struct=Get Biped_Raptor_Base_Struct())
   Output Animation Pose(Result=Modify Curve(SourcePose=Sequence Player(Sequence=@n1.SitToStand_87_#12) {bLoopAnimation=false}, CurveValues_0=0, Alpha=1) {curveNames [AlphaCurve], curveValues [0]}) {
    node {stateExitFunction {functionName=OnStateExit}}
    stateExitFunction {bSelfContext=true, memberGuid=#26, memberName=ToSitting_Exit}
   }
  ]
 }
 SitPose {
  path=AnimGraph.AnimGraphNode_StateMachine_0.Idle.AnimStateNode_5.SitPose
  nodes [
   n1 = Break Biped Raptor Base Struct(Biped_Raptor_Base_Struct=Get Biped_Raptor_Base_Struct())
   Output Animation Pose(Result=Modify Curve(SourcePose=Sequence Player(Sequence=@n1.Sitting_88_#11), CurveValues_0=0, Alpha=1) {curveNames [AlphaCurve], curveValues [0]}) {
    node {stateEntryFunction {functionName=ToSitting_Entry}}
    stateEntryFunction {bSelfContext=true, memberGuid=#25, memberName=ToSitting_Entry}
   }
  ]
 }
 StandingPose {
  path=AnimGraph.AnimGraphNode_StateMachine_0.Idle.AnimStateNode_6.StandingPose
  nodes [
   AnimGraphNode_StateMachine_0 = New State Machine() {is=title AnimGraphNode_StateMachine}
   n1 = Break Biped Raptor Base Struct(Biped_Raptor_Base_Struct=Get Biped_Raptor_Base_Struct())
   Output Animation Pose(Result=Apply Additive(
     Base=Sequence Player(Sequence=@n1.Locomotion_Idle_72_#2)
     Additive=Sequence Player(Sequence=@n1.FeatherSpreadPose_96_#14) {bLoopAnimation=false}
    ) {alphaBoolBlend {blendOption=Linear}})
  ]
 }
 New State Machine {
  path=AnimGraph.AnimGraphNode_StateMachine_0.Idle.AnimStateNode_6.StandingPose.AnimGraphNode_StateMachine_0.New State Machine
  links []
  nodes []
 }
 AnimationTransitionGraph_0 {path=AnimGraph.AnimGraphNode_StateMachine_0.Idle.AnimStateTransitionNode_0.AnimationTransitionGraph_0, nodes [Result()]}
 AnimationTransitionGraph_0 {
  path=AnimGraph.AnimGraphNode_StateMachine_0.Idle.AnimStateTransitionNode_1.AnimationTransitionGraph_0
  nodes [Result(bCanEnterTransition=Get bIsStandToSit() -> (bIsStandToSit:bool))]
 }
 AnimationTransitionGraph_0 {
  path=AnimGraph.AnimGraphNode_StateMachine_0.Idle.AnimStateTransitionNode_10.AnimationTransitionGraph_0
  nodes [Result(bCanEnterTransition=Get bIsSitting() -> (bIsSitting:bool))]
 }
 Transition {
  path=AnimGraph.AnimGraphNode_StateMachine_0.Idle.AnimStateTransitionNode_11.Transition
  nodes [Result(bCanEnterTransition=NOT Boolean(A=Get bIsMoodCombat() -> (bIsMoodCombat:bool)))]
 }
 AnimationTransitionGraph_3 {
  path=AnimGraph.AnimGraphNode_StateMachine_0.Idle.AnimStateTransitionNode_2.AnimationTransitionGraph_3
  nodes [Result(bCanEnterTransition=Get bIsMoodCombat())]
 }
 AnimationTransitionGraph_3 {path=AnimGraph.AnimGraphNode_StateMachine_0.Idle.AnimStateTransitionNode_3.AnimationTransitionGraph_3, nodes [Result()]}
 AnimationTransitionGraph_3 {
  path=AnimGraph.AnimGraphNode_StateMachine_0.Idle.AnimStateTransitionNode_4.AnimationTransitionGraph_3
  nodes [Result(bCanEnterTransition=NOT Boolean(A=Get bIsMoodCombat()))]
 }
 AnimationTransitionGraph_3 {path=AnimGraph.AnimGraphNode_StateMachine_0.Idle.AnimStateTransitionNode_5.AnimationTransitionGraph_3, nodes [Result()]}
 AnimationTransitionGraph_0 {path=AnimGraph.AnimGraphNode_StateMachine_0.Idle.AnimStateTransitionNode_6.AnimationTransitionGraph_0, nodes [Result()]}
 AnimationTransitionGraph_0 {
  path=AnimGraph.AnimGraphNode_StateMachine_0.Idle.AnimStateTransitionNode_7.AnimationTransitionGraph_0
  nodes [Result(bCanEnterTransition=true)]
 }
 AnimationTransitionGraph_0 {
  path=AnimGraph.AnimGraphNode_StateMachine_0.Idle.AnimStateTransitionNode_8.AnimationTransitionGraph_0
  nodes [Result(bCanEnterTransition=Get bIsSitToStand() -> (bIsSitToStand:bool))]
 }
 AnimationTransitionGraph_0 {
  path=AnimGraph.AnimGraphNode_StateMachine_0.Idle.AnimStateTransitionNode_9.AnimationTransitionGraph_0
  nodes [Result(bCanEnterTransition=true)]
 }
 Body Pose Additive {
  entry=Conduit
  path=AnimGraph.AnimGraphNode_StateMachine_1.Body Pose Additive
  ids=id
  nodes [
   AnimStateConduitNode_0 = Conduit()
   AnimStateNode_0 = BodyPoseChooser() {is=title AnimStateNode}
   AnimStateNode_2 = BodyDefaultPose() {is=title AnimStateNode}
   AnimStateTransitionNode_0 = AnimStateTransitionNode() {
    crossfadeDuration=0.4
    logicType=TLT_Inertialization
    blendProfileWrapper {blendProfile=&CrescentNail_Biped_Skeleton:FastHeadSlowBody}
   }
   AnimStateTransitionNode_2 = AnimStateTransitionNode() {
    crossfadeDuration=0.6
    logicType=TLT_Inertialization
    priorityOrder=100
    blendProfileWrapper {blendProfile=&CrescentNail_Biped_Skeleton:FastHeadSlowBody}
   }
   AnimStateTransitionNode_3 = AnimStateTransitionNode() {bAutomaticRuleBasedOnSequencePlayerInState=true}
   AnimStateTransitionNode_5 = AnimStateTransitionNode()
  ]
  links [
   {
    dst=Conduit
    rule=AnimGraph.AnimGraphNode_StateMachine_1.Body Pose Additive.AnimStateTransitionNode_5.Transition
    src=BodyDefaultPose
    via=AnimStateTransitionNode_5
   }
   {
    dst=Conduit
    rule=AnimGraph.AnimGraphNode_StateMachine_1.Body Pose Additive.AnimStateTransitionNode_3.Transition
    src=BodyPoseChooser
    via=AnimStateTransitionNode_3
   }
   {
    dst=BodyDefaultPose
    rule=AnimGraph.AnimGraphNode_StateMachine_1.Body Pose Additive.AnimStateTransitionNode_2.Transition
    src=Conduit
    via=AnimStateTransitionNode_2
   }
   {
    dst=BodyPoseChooser
    rule=AnimGraph.AnimGraphNode_StateMachine_1.Body Pose Additive.AnimStateTransitionNode_0.Transition
    src=Conduit
    via=AnimStateTransitionNode_0
   }
  ]
 }
 Conduit {path=AnimGraph.AnimGraphNode_StateMachine_1.Body Pose Additive.AnimStateConduitNode_0.Conduit, nodes [Result(bCanEnterTransition=true)]}
 BodyPoseChooser {
  path=AnimGraph.AnimGraphNode_StateMachine_1.Body Pose Additive.AnimStateNode_0.BodyPoseChooser
  nodes [
   AnimGraphNode_ChooserPlayer_1 = Chooser Player: PRA_Companion_BodyPose!() -> (Pose:Pose Link Structure) {
    is=title AnimGraphNode_ChooserPlayer
    chooser {
     _structType=/Script/ProxyTable.LookupProxy
     proxy=/CosmeticCompanions/Chooser/PRA_Companion_BodyPose
     proxyTable {
      _structType=/Script/ProxyTable.ProxyTableContextProperty
      binding {contextIndex=0, displayName=Chooser Tables, isBoundToRoot=false, propertyBindingChain [ChooserTables]}
     }
    }
    chooserContextDefinition [
     {
      _structType=/Script/Chooser.ContextObjectTypeClass
      class=&/Biped_Raptor_Base_AnimBP.Biped_Raptor_Base_AnimBP_C
      direction=ReadWrite
     }
     {struct=/Script/Chooser.ChooserPlayerSettings, _structType=/Script/Chooser.ContextObjectTypeStruct, direction=Read}
    ]
   }
   Output Animation Pose(Result=@AnimGraphNode_ChooserPlayer_1)
  ]
 }
 AnimationBlendStackGraph_0 {
  path=AnimGraph.AnimGraphNode_StateMachine_1.Body Pose Additive.AnimStateNode_0.BodyPoseChooser.AnimGraphNode_ChooserPlayer_1.AnimationBlendStackGraph_0
  nodes [Output Pose!(Result:Pose Link Structure=Blend Stack Input!() -> (Pose:Pose Link Structure) {is=title AnimGraphNode_BlendStackInput}) {is=title AnimGraphNode_BlendStackResult}]
 }
 BodyDefaultPose {
  path=AnimGraph.AnimGraphNode_StateMachine_1.Body Pose Additive.AnimStateNode_2.BodyDefaultPose
  nodes [Output Animation Pose(Result=Additive Identity Pose())]
 }
 Transition {
  path=AnimGraph.AnimGraphNode_StateMachine_1.Body Pose Additive.AnimStateTransitionNode_0.Transition
  nodes [Result(bCanEnterTransition=Break Companion Random Number Timer!(CompanionRandomNumberTimer:"Companion Random Number Timer Structure (by ref)"=Get AnticsTimer() -> (AnticsTimer:Companion Random Number Timer Structure)) -> (ValueInt:int, ValueFloat:float=~, bJustChanged:bool=~) {is=title K2Node_BreakStruct}.bJustChanged)]
 }
 Transition {
  path=AnimGraph.AnimGraphNode_StateMachine_1.Body Pose Additive.AnimStateTransitionNode_2.Transition
  nodes [Result(bCanEnterTransition=true)]
 }
 Transition {path=AnimGraph.AnimGraphNode_StateMachine_1.Body Pose Additive.AnimStateTransitionNode_3.Transition, nodes [Result()]}
 Transition {
  path=AnimGraph.AnimGraphNode_StateMachine_1.Body Pose Additive.AnimStateTransitionNode_5.Transition
  nodes [Result(bCanEnterTransition=true)]
 }
 Locomotion {
  entry=LocomotionState
  path=AnimGraph.AnimGraphNode_StateMachine_2.Locomotion
  ids=id
  nodes [AnimStateNode_0 = LocomotionState() {is=title AnimStateNode}]
 }
 LocomotionState {
  path=AnimGraph.AnimGraphNode_StateMachine_2.Locomotion.AnimStateNode_0.LocomotionState
  nodes [
   n1 = Break Biped Raptor Base Struct(Biped_Raptor_Base_Struct=Get Biped_Raptor_Base_Struct())
   n2 = Sequence Player(Sequence=@n1.Locomotion_Walk_76_#6) {groupName=Locomotion, method=SyncGroup}
   n3 = Blend Poses by bool(
    BlendPose_0=Sequence Player(Sequence=@n1.Locomotion_Run_74_#4) {groupName=Locomotion, method=SyncGroup}
    BlendPose_1=Blend Poses by bool(
     BlendPose_0=Sequence Player(Sequence=@n1.Locomotion_Jog_75_#5) {groupName=Locomotion, method=SyncGroup}
     BlendPose_1=Modify Curve(SourcePose=@n2, CurveValues_0=1, Alpha=1) {curveNames [Walk], curveValues [0]}
     BlendTime_0=0.2
     BlendTime_1=0.2
     bActiveValue=Get bIsLocoJog() -> (bIsLocoJog:bool)
    ) {blendType=Sinusoidal, transitionType=Inertialization}
    BlendTime_0=0.2
    BlendTime_1=0.2
    bActiveValue=Get bIsLocoRun() -> (bIsLocoRun:bool)
   ) {blendType=Sinusoidal, transitionType=Inertialization}
   n4 = Sequence Player(Sequence=@n1.Locomotion_Sprint_73_#3) {groupName=Locomotion, method=SyncGroup}
   n5 = Get bIsSprinting()
   Output Animation Pose(Result=Modify Curve(
     SourcePose=Blend Poses by bool(BlendPose_0=@n4, BlendPose_1=@n3, BlendTime_0=0.2, BlendTime_1=0.2, bActiveValue=@n5) {blendType=Sinusoidal, transitionType=Inertialization}
     CurveValues_0=1
     Alpha=1
    ) {curveNames [Locomotion], curveValues [0]})
  ]
 }
 BodyMoodAdditive {
  entry=Default
  path=AnimGraph.AnimGraphNode_StateMachine_3.BodyMoodAdditive
  ids=id
  nodes [
   AnimStateNode_0 = Worried() {is=title AnimStateNode}
   AnimStateNode_1 = Default() {is=title AnimStateNode}
   AnimStateNode_2 = Angry() {is=title AnimStateNode}
   AnimStateTransitionNode_0 = AnimStateTransitionNode() {logicType=TLT_Inertialization}
   AnimStateTransitionNode_1 = AnimStateTransitionNode() {logicType=TLT_Inertialization, priorityOrder=100}
   AnimStateTransitionNode_4 = AnimStateTransitionNode() {bAllowInertializationForSelfTransitions=true}
   AnimStateTransitionNode_5 = AnimStateTransitionNode() {bAllowInertializationForSelfTransitions=true}
  ]
  links [
   {
    dst=Default
    rule=AnimGraph.AnimGraphNode_StateMachine_3.BodyMoodAdditive.AnimStateTransitionNode_5.Transition
    src=Angry
    via=AnimStateTransitionNode_5
   }
   {
    dst=Angry
    rule=AnimGraph.AnimGraphNode_StateMachine_3.BodyMoodAdditive.AnimStateTransitionNode_4.Transition
    src=Default
    via=AnimStateTransitionNode_4
   }
   {
    dst=Worried
    rule=AnimGraph.AnimGraphNode_StateMachine_3.BodyMoodAdditive.AnimStateTransitionNode_0.Transition
    src=Default
    via=AnimStateTransitionNode_0
   }
   {
    dst=Default
    rule=AnimGraph.AnimGraphNode_StateMachine_3.BodyMoodAdditive.AnimStateTransitionNode_1.Transition
    src=Worried
    via=AnimStateTransitionNode_1
   }
  ]
 }
 Worried {
  path=AnimGraph.AnimGraphNode_StateMachine_3.BodyMoodAdditive.AnimStateNode_0.Worried
  nodes [
   n1 = Get Biped_Raptor_Base_Struct()
   Output Animation Pose(Result=Sequence Player(Sequence=Break Biped Raptor Base Struct(Biped_Raptor_Base_Struct=@n1).BodyPose_Worried_77_#8))
  ]
 }
 Default {
  path=AnimGraph.AnimGraphNode_StateMachine_3.BodyMoodAdditive.AnimStateNode_1.Default
  nodes [Output Animation Pose(Result=Additive Identity Pose())]
 }
 Angry {
  path=AnimGraph.AnimGraphNode_StateMachine_3.BodyMoodAdditive.AnimStateNode_2.Angry
  nodes [
   n1 = Get Biped_Raptor_Base_Struct()
   Output Animation Pose(Result=Sequence Player(Sequence=Break Biped Raptor Base Struct(Biped_Raptor_Base_Struct=@n1).BodyPose_Angry_108_#18))
  ]
 }
 Transition {
  path=AnimGraph.AnimGraphNode_StateMachine_3.BodyMoodAdditive.AnimStateTransitionNode_0.Transition
  nodes [Result(bCanEnterTransition=Get bIsMoodWorried() -> (bIsMoodWorried:bool))]
 }
 Transition {
  path=AnimGraph.AnimGraphNode_StateMachine_3.BodyMoodAdditive.AnimStateTransitionNode_1.Transition
  nodes [Result(bCanEnterTransition=NOT Boolean(A=Get bIsMoodWorried()))]
 }
 Transition {
  path=AnimGraph.AnimGraphNode_StateMachine_3.BodyMoodAdditive.AnimStateTransitionNode_4.Transition
  nodes [Result(bCanEnterTransition=Get bPlayCombatBodyAdditive() -> (bPlayCombatBodyAdditive:bool))]
 }
 Transition {
  path=AnimGraph.AnimGraphNode_StateMachine_3.BodyMoodAdditive.AnimStateTransitionNode_5.Transition
  nodes [Result(bCanEnterTransition=NOT Boolean(A=Get bPlayCombatBodyAdditive()))]
 }
 LowFidelityStateMachine {
  entry=Conduit
  path=AnimGraph.AnimGraphNode_StateMachine_4.LowFidelityStateMachine
  ids=id
  nodes [
   AnimStateConduitNode_0 = Conduit()
   AnimStateNode_0 = EmoteLoops() {is=title AnimStateNode}
   AnimStateNode_1 = TransitionToDefaultIdle() {is=title AnimStateNode}
   AnimStateNode_4 = Locomotion() {is=title AnimStateNode}
   AnimStateNode_5 = Idle() {is=title AnimStateNode, bAlwaysResetOnEntry=true}
   AnimStateTransitionNode_1 = AnimStateTransitionNode() {crossfadeDuration=0.35, priorityOrder=100}
   AnimStateTransitionNode_10 = AnimStateTransitionNode() {logicType=TLT_Inertialization}
   AnimStateTransitionNode_11 = AnimStateTransitionNode() {logicType=TLT_Inertialization}
   AnimStateTransitionNode_12 = AnimStateTransitionNode() {logicType=TLT_Inertialization, priorityOrder=2}
   AnimStateTransitionNode_13 = AnimStateTransitionNode()
   AnimStateTransitionNode_14 = AnimStateTransitionNode()
   AnimStateTransitionNode_15 = AnimStateTransitionNode() {bAutomaticRuleBasedOnSequencePlayerInState=true}
   AnimStateTransitionNode_16 = AnimStateTransitionNode() {
    crossfadeDuration=0.3
    logicType=TLT_Inertialization
    blendProfileWrapper {blendProfile=&CrescentNail_Biped_Skeleton:FastFeetSlowHead}
   }
   AnimStateTransitionNode_2 = AnimStateTransitionNode()
   AnimStateTransitionNode_22 = AnimStateTransitionNode()
  ]
  links [
   {
    dst=EmoteLoops
    rule=AnimGraph.AnimGraphNode_StateMachine_4.LowFidelityStateMachine.AnimStateTransitionNode_10.AnimationTransitionGraph_0
    src=Conduit
    via=AnimStateTransitionNode_10
   }
   {
    dst=Idle
    rule=AnimGraph.AnimGraphNode_StateMachine_4.LowFidelityStateMachine.AnimStateTransitionNode_1.AnimationTransitionGraph_0
    src=Conduit
    via=AnimStateTransitionNode_1
   }
   {
    dst=Locomotion
    rule=AnimGraph.AnimGraphNode_StateMachine_4.LowFidelityStateMachine.AnimStateTransitionNode_16.AnimationTransitionGraph_0
    src=Conduit
    via=AnimStateTransitionNode_16
   }
   {
    dst=TransitionToDefaultIdle
    rule=AnimGraph.AnimGraphNode_StateMachine_4.LowFidelityStateMachine.AnimStateTransitionNode_12.AnimationTransitionGraph_0
    src=Conduit
    via=AnimStateTransitionNode_12
   }
   {
    dst=Conduit
    rule=AnimGraph.AnimGraphNode_StateMachine_4.LowFidelityStateMachine.AnimStateTransitionNode_14.AnimationTransitionGraph_0
    src=EmoteLoops
    via=AnimStateTransitionNode_14
   }
   {
    dst=TransitionToDefaultIdle
    rule=AnimGraph.AnimGraphNode_StateMachine_4.LowFidelityStateMachine.AnimStateTransitionNode_11.AnimationTransitionGraph_0
    src=EmoteLoops
    via=AnimStateTransitionNode_11
   }
   {
    dst=Conduit
    rule=AnimGraph.AnimGraphNode_StateMachine_4.LowFidelityStateMachine.AnimStateTransitionNode_2.AnimationTransitionGraph_0
    src=Idle
    via=AnimStateTransitionNode_2
   }
   {
    dst=Conduit
    rule=AnimGraph.AnimGraphNode_StateMachine_4.LowFidelityStateMachine.AnimStateTransitionNode_22.Transition
    src=Locomotion
    via=AnimStateTransitionNode_22
   }
   {
    dst=Conduit
    rule=AnimGraph.AnimGraphNode_StateMachine_4.LowFidelityStateMachine.AnimStateTransitionNode_13.AnimationTransitionGraph_0
    src=TransitionToDefaultIdle
    via=AnimStateTransitionNode_13
   }
   {
    dst=Conduit
    rule=AnimGraph.AnimGraphNode_StateMachine_4.LowFidelityStateMachine.AnimStateTransitionNode_15.AnimationTransitionGraph_0
    src=TransitionToDefaultIdle
    via=AnimStateTransitionNode_15
   }
  ]
 }
 Conduit {path=AnimGraph.AnimGraphNode_StateMachine_4.LowFidelityStateMachine.AnimStateConduitNode_0.Conduit, nodes [Result(bCanEnterTransition=true)]}
 EmoteLoops {
  path=AnimGraph.AnimGraphNode_StateMachine_4.LowFidelityStateMachine.AnimStateNode_0.EmoteLoops
  nodes [
   AnimGraphNode_ChooserPlayer_1 = Chooser Player: PRA_Companion_SpecialIdle!() -> (Pose:Pose Link Structure) {
    is=title AnimGraphNode_ChooserPlayer
    chooser {
     _structType=&LookupProxy
     proxy=/CosmeticCompanions/Chooser/PRA_Companion_SpecialIdle
     proxyTable {
      _structType=&ProxyTableContextProperty
      binding {contextIndex=0, displayName=Chooser Tables, isBoundToRoot=false, propertyBindingChain [ChooserTables]}
     }
    }
    chooserContextDefinition [
     {
      _structType=&ContextObjectTypeClass
      class=&Biped_Raptor_Base_AnimBP_C
      direction=ReadWrite
     }
     {struct=&ChooserPlayerSettings, _structType=&ContextObjectTypeStruct, direction=Read}
    ]
   }
   Output Animation Pose(Result=@AnimGraphNode_ChooserPlayer_1)
  ]
 }
 AnimationBlendStackGraph_0 {
  path=AnimGraph.AnimGraphNode_StateMachine_4.LowFidelityStateMachine.AnimStateNode_0.EmoteLoops.AnimGraphNode_ChooserPlayer_1.AnimationBlendStackGraph_0
  nodes [Output Pose(Result=Blend Stack Input())]
 }
 TransitionToDefaultIdle {
  path=AnimGraph.AnimGraphNode_StateMachine_4.LowFidelityStateMachine.AnimStateNode_1.TransitionToDefaultIdle
  nodes [
   AnimGraphNode_ChooserPlayer_1 = Chooser Player: PRA_Companion_ReactToIdle!() -> (Pose:Pose Link Structure) {
    is=title AnimGraphNode_ChooserPlayer
    chooser {
     _structType=&LookupProxy
     proxy=/CosmeticCompanions/Chooser/PRA_Companion_ReactToIdle
     proxyTable {
      _structType=&ProxyTableContextProperty
      binding {contextIndex=0, displayName=Chooser Tables, isBoundToRoot=false, propertyBindingChain [ChooserTables]}
     }
    }
    chooserContextDefinition [
     {
      _structType=&ContextObjectTypeClass
      class=&Biped_Raptor_Base_AnimBP_C
      direction=ReadWrite
     }
     {struct=&ChooserPlayerSettings, _structType=&ContextObjectTypeStruct, direction=Read}
    ]
   }
   Output Animation Pose(Result=@AnimGraphNode_ChooserPlayer_1) {
    node {stateEntryFunction {functionName=OnStateEntry_0}, stateExitFunction {functionName=OnStateExit_0}}
    stateEntryFunction {bSelfContext=true, memberGuid=#27, memberName=OnStateEntry_0}
    stateExitFunction {bSelfContext=true, memberGuid=#28, memberName=OnStateExit_0}
   }
  ]
 }
 AnimationBlendStackGraph_0 {
  path=AnimGraph.AnimGraphNode_StateMachine_4.LowFidelityStateMachine.AnimStateNode_1.TransitionToDefaultIdle.AnimGraphNode_ChooserPlayer_1.AnimationBlendStackGraph_0
  nodes [Output Pose(Result=Blend Stack Input())]
 }
 Locomotion {
  path=AnimGraph.AnimGraphNode_StateMachine_4.LowFidelityStateMachine.AnimStateNode_4.Locomotion
  nodes [Output Animation Pose(Result=Use cached pose 'LocomotionPose'())]
 }
 Idle {
  path=AnimGraph.AnimGraphNode_StateMachine_4.LowFidelityStateMachine.AnimStateNode_5.Idle
  nodes [Output Animation Pose(Result=Use cached pose 'IdlePose'())]
 }
 AnimationTransitionGraph_0 {
  path=AnimGraph.AnimGraphNode_StateMachine_4.LowFidelityStateMachine.AnimStateTransitionNode_1.AnimationTransitionGraph_0
  nodes [Result(bCanEnterTransition=true)]
 }
 AnimationTransitionGraph_0 {
  path=AnimGraph.AnimGraphNode_StateMachine_4.LowFidelityStateMachine.AnimStateTransitionNode_10.AnimationTransitionGraph_0
  nodes [Result(bCanEnterTransition=Get bEnterEmoteLoops() -> (bEnterEmoteLoops:bool))]
 }
 AnimationTransitionGraph_0 {
  path=AnimGraph.AnimGraphNode_StateMachine_4.LowFidelityStateMachine.AnimStateTransitionNode_11.AnimationTransitionGraph_0
  nodes [Result(bCanEnterTransition=NOT Boolean(A=Get bIsPlayingEmote()))]
 }
 AnimationTransitionGraph_0 {
  path=AnimGraph.AnimGraphNode_StateMachine_4.LowFidelityStateMachine.AnimStateTransitionNode_12.AnimationTransitionGraph_0
  nodes [Result(bCanEnterTransition=Get bIsReactionHappyPlaying() -> (bIsReactionHappyPlaying:bool))]
 }
 AnimationTransitionGraph_0 {
  path=AnimGraph.AnimGraphNode_StateMachine_4.LowFidelityStateMachine.AnimStateTransitionNode_13.AnimationTransitionGraph_0
  nodes [Result(bCanEnterTransition=Get bIsMoving() -> (bIsMoving:bool))]
 }
 AnimationTransitionGraph_0 {
  path=AnimGraph.AnimGraphNode_StateMachine_4.LowFidelityStateMachine.AnimStateTransitionNode_14.AnimationTransitionGraph_0
  nodes [Result(bCanEnterTransition=Get bIsMoving())]
 }
 AnimationTransitionGraph_0 {path=AnimGraph.AnimGraphNode_StateMachine_4.LowFidelityStateMachine.AnimStateTransitionNode_15.AnimationTransitionGraph_0, nodes [Result()]}
 AnimationTransitionGraph_0 {
  path=AnimGraph.AnimGraphNode_StateMachine_4.LowFidelityStateMachine.AnimStateTransitionNode_16.AnimationTransitionGraph_0
  nodes [Result(bCanEnterTransition=Get bIsMoving())]
 }
 AnimationTransitionGraph_0 {
  path=AnimGraph.AnimGraphNode_StateMachine_4.LowFidelityStateMachine.AnimStateTransitionNode_2.AnimationTransitionGraph_0
  nodes [Result(bCanEnterTransition=true)]
 }
 Transition {
  path=AnimGraph.AnimGraphNode_StateMachine_4.LowFidelityStateMachine.AnimStateTransitionNode_22.Transition
  nodes [Result(bCanEnterTransition=true)]
 }
 LocomotionStartAdditive {
  entry=IdentPose
  path=AnimGraph.AnimGraphNode_StateMachine_5.LocomotionStartAdditive
  ids=id
  nodes [
   AnimStateNode_0 = IdentPose() {is=title AnimStateNode}
   AnimStateNode_1 = Start() {is=title AnimStateNode}
   AnimStateTransitionNode_0 = AnimStateTransitionNode() {crossfadeDuration=0.1}
   AnimStateTransitionNode_1 = AnimStateTransitionNode() {bAutomaticRuleBasedOnSequencePlayerInState=true}
  ]
  links [
   {
    dst=Start
    rule=AnimGraph.AnimGraphNode_StateMachine_5.LocomotionStartAdditive.AnimStateTransitionNode_0.Transition
    src=IdentPose
    via=AnimStateTransitionNode_0
   }
   {
    dst=IdentPose
    rule=AnimGraph.AnimGraphNode_StateMachine_5.LocomotionStartAdditive.AnimStateTransitionNode_1.Transition
    src=Start
    via=AnimStateTransitionNode_1
   }
  ]
 }
 IdentPose {
  path=AnimGraph.AnimGraphNode_StateMachine_5.LocomotionStartAdditive.AnimStateNode_0.IdentPose
  nodes [Output Animation Pose(Result=Additive Identity Pose())]
 }
 Start {
  path=AnimGraph.AnimGraphNode_StateMachine_5.LocomotionStartAdditive.AnimStateNode_1.Start
  nodes [
   Layered blend per bone(BlendPoses_0=Additive Identity Pose()) {
    blendMode=BlendMask
    blendMasks [&CrescentNail_Biped_Skeleton:NoLegsOrArms]
    layerSetup []
   }
   n1 = Get Biped_Raptor_Base_Struct()
   Output Animation Pose(Result=Sequence Player(Sequence=Break Biped Raptor Base Struct(Biped_Raptor_Base_Struct=@n1)) {bLoopAnimation=false})
  ]
 }
 Transition {
  path=AnimGraph.AnimGraphNode_StateMachine_5.LocomotionStartAdditive.AnimStateTransitionNode_0.Transition
  nodes [Result(bCanEnterTransition=Get bIdentPoseToStart() -> (bIdentPoseToStart:bool))]
 }
 Transition {path=AnimGraph.AnimGraphNode_StateMachine_5.LocomotionStartAdditive.AnimStateTransitionNode_1.Transition, nodes [Result()]}
 Head Antic {
  entry=Conduit
  path=AnimGraph.AnimGraphNode_StateMachine_6.Head Antic
  ids=id
  nodes [
   AnimStateConduitNode_0 = Conduit()
   AnimStateNode_0 = AnticDefaultPose() {is=title AnimStateNode}
   AnimStateNode_1 = AnticChooser() {is=title AnimStateNode}
   AnimStateTransitionNode_0 = AnimStateTransitionNode() {logicType=TLT_Inertialization, priorityOrder=100}
   AnimStateTransitionNode_1 = AnimStateTransitionNode() {logicType=TLT_Inertialization}
   AnimStateTransitionNode_2 = AnimStateTransitionNode() {bAutomaticRuleBasedOnSequencePlayerInState=true}
   AnimStateTransitionNode_5 = AnimStateTransitionNode()
  ]
  links [
   {
    dst=Conduit
    rule=AnimGraph.AnimGraphNode_StateMachine_6.Head Antic.AnimStateTransitionNode_2.Transition
    src=AnticChooser
    via=AnimStateTransitionNode_2
   }
   {
    dst=Conduit
    rule=AnimGraph.AnimGraphNode_StateMachine_6.Head Antic.AnimStateTransitionNode_5.Transition
    src=AnticDefaultPose
    via=AnimStateTransitionNode_5
   }
   {
    dst=AnticChooser
    rule=AnimGraph.AnimGraphNode_StateMachine_6.Head Antic.AnimStateTransitionNode_1.Transition
    src=Conduit
    via=AnimStateTransitionNode_1
   }
   {
    dst=AnticDefaultPose
    rule=AnimGraph.AnimGraphNode_StateMachine_6.Head Antic.AnimStateTransitionNode_0.Transition
    src=Conduit
    via=AnimStateTransitionNode_0
   }
  ]
 }
 Conduit {path=AnimGraph.AnimGraphNode_StateMachine_6.Head Antic.AnimStateConduitNode_0.Conduit, nodes [Result(bCanEnterTransition=true)]}
 AnticDefaultPose {
  path=AnimGraph.AnimGraphNode_StateMachine_6.Head Antic.AnimStateNode_0.AnticDefaultPose
  nodes [Output Animation Pose(Result=Additive Identity Pose())]
 }
 AnticChooser {
  path=AnimGraph.AnimGraphNode_StateMachine_6.Head Antic.AnimStateNode_1.AnticChooser
  nodes [Output Animation Pose(Result=Sequence Player() {bLoopAnimation=false})]
 }
 Transition {path=AnimGraph.AnimGraphNode_StateMachine_6.Head Antic.AnimStateTransitionNode_0.Transition, nodes [Result(bCanEnterTransition=true)]}
 Transition {
  path=AnimGraph.AnimGraphNode_StateMachine_6.Head Antic.AnimStateTransitionNode_1.Transition
  nodes [
   Result(bCanEnterTransition=Get!() -> (AnticSequenceInfo_AnticSequenceToPlay:Anim Sequence Object Reference, AnticSequenceInfo_bAnticSequenceReady:bool=~).AnticSequenceInfo_bAnticSequenceReady)
   Break Companion Random Number Timer(CompanionRandomNumberTimer=Get AnticsTimer()) -> (ValueFloat=~, bJustChanged=~)
  ]
 }
 Transition {path=AnimGraph.AnimGraphNode_StateMachine_6.Head Antic.AnimStateTransitionNode_2.Transition, nodes [Result()]}
 Transition {path=AnimGraph.AnimGraphNode_StateMachine_6.Head Antic.AnimStateTransitionNode_5.Transition, nodes [Result(bCanEnterTransition=true)]}
 LocomotionStateMachine {
  entry=Conduit
  path=AnimGraph.AnimGraphNode_StateMachine_7.LocomotionStateMachine
  ids=id
  nodes [
   AnimStateConduitNode_0 = Conduit()
   AnimStateNode_0 = EmoteLoops()
   AnimStateNode_1 = TransitionToDefaultIdle()
   AnimStateNode_2 = LocoStop() {is=title AnimStateNode}
   AnimStateNode_4 = Locomotion()
   AnimStateNode_5 = Idle() {bAlwaysResetOnEntry=true}
   AnimStateTransitionNode_0 = AnimStateTransitionNode() {bAutomaticRuleBasedOnSequencePlayerInState=true}
   AnimStateTransitionNode_1 = AnimStateTransitionNode() {crossfadeDuration=0.35, priorityOrder=100}
   AnimStateTransitionNode_10 = AnimStateTransitionNode() {logicType=TLT_Inertialization}
   AnimStateTransitionNode_11 = AnimStateTransitionNode() {logicType=TLT_Inertialization}
   AnimStateTransitionNode_12 = AnimStateTransitionNode() {logicType=TLT_Inertialization, priorityOrder=2}
   AnimStateTransitionNode_13 = AnimStateTransitionNode()
   AnimStateTransitionNode_14 = AnimStateTransitionNode()
   AnimStateTransitionNode_15 = AnimStateTransitionNode() {bAutomaticRuleBasedOnSequencePlayerInState=true}
   AnimStateTransitionNode_16 = AnimStateTransitionNode() {
    crossfadeDuration=0.3
    logicType=TLT_Inertialization
    blendProfileWrapper {blendProfile=&CrescentNail_Biped_Skeleton:FastFeetSlowHead}
   }
   AnimStateTransitionNode_17 = AnimStateTransitionNode() {logicType=TLT_Inertialization}
   AnimStateTransitionNode_2 = AnimStateTransitionNode()
   AnimStateTransitionNode_3 = AnimStateTransitionNode() {bAllowInertializationForSelfTransitions=true}
   AnimStateTransitionNode_4 = AnimStateTransitionNode() {
    logicType=TLT_Inertialization
    blendProfileWrapper {blendProfile=&CrescentNail_Biped_Skeleton:FastFeetSlowHead}
   }
  ]
  links [
   {
    dst=EmoteLoops
    rule=AnimGraph.AnimGraphNode_StateMachine_7.LocomotionStateMachine.AnimStateTransitionNode_10.AnimationTransitionGraph_0
    src=Conduit
    via=AnimStateTransitionNode_10
   }
   {
    dst=Idle
    rule=AnimGraph.AnimGraphNode_StateMachine_7.LocomotionStateMachine.AnimStateTransitionNode_1.AnimationTransitionGraph_0
    src=Conduit
    via=AnimStateTransitionNode_1
   }
   {
    dst=Locomotion
    rule=AnimGraph.AnimGraphNode_StateMachine_7.LocomotionStateMachine.AnimStateTransitionNode_16.AnimationTransitionGraph_0
    src=Conduit
    via=AnimStateTransitionNode_16
   }
   {
    dst=TransitionToDefaultIdle
    rule=AnimGraph.AnimGraphNode_StateMachine_7.LocomotionStateMachine.AnimStateTransitionNode_12.AnimationTransitionGraph_0
    src=Conduit
    via=AnimStateTransitionNode_12
   }
   {
    dst=Conduit
    rule=AnimGraph.AnimGraphNode_StateMachine_7.LocomotionStateMachine.AnimStateTransitionNode_14.AnimationTransitionGraph_0
    src=EmoteLoops
    via=AnimStateTransitionNode_14
   }
   {
    dst=TransitionToDefaultIdle
    rule=AnimGraph.AnimGraphNode_StateMachine_7.LocomotionStateMachine.AnimStateTransitionNode_11.AnimationTransitionGraph_0
    src=EmoteLoops
    via=AnimStateTransitionNode_11
   }
   {
    dst=Conduit
    rule=AnimGraph.AnimGraphNode_StateMachine_7.LocomotionStateMachine.AnimStateTransitionNode_2.AnimationTransitionGraph_0
    src=Idle
    via=AnimStateTransitionNode_2
   }
   {
    dst=Conduit
    rule=AnimGraph.AnimGraphNode_StateMachine_7.LocomotionStateMachine.AnimStateTransitionNode_0.AnimationTransitionGraph_0
    src=LocoStop
    via=AnimStateTransitionNode_0
   }
   {
    dst=Conduit
    rule=AnimGraph.AnimGraphNode_StateMachine_7.LocomotionStateMachine.AnimStateTransitionNode_3.Transition
    src=LocoStop
    via=AnimStateTransitionNode_3
   }
   {
    dst=Locomotion
    rule=AnimGraph.AnimGraphNode_StateMachine_7.LocomotionStateMachine.AnimStateTransitionNode_17.AnimationTransitionGraph_0
    src=LocoStop
    via=AnimStateTransitionNode_17
   }
   {
    dst=LocoStop
    rule=AnimGraph.AnimGraphNode_StateMachine_7.LocomotionStateMachine.AnimStateTransitionNode_4.AnimationTransitionGraph_0
    src=Locomotion
    via=AnimStateTransitionNode_4
   }
   {
    dst=Conduit
    rule=AnimGraph.AnimGraphNode_StateMachine_7.LocomotionStateMachine.AnimStateTransitionNode_13.AnimationTransitionGraph_0
    src=TransitionToDefaultIdle
    via=AnimStateTransitionNode_13
   }
   {
    dst=Conduit
    rule=AnimGraph.AnimGraphNode_StateMachine_7.LocomotionStateMachine.AnimStateTransitionNode_15.AnimationTransitionGraph_0
    src=TransitionToDefaultIdle
    via=AnimStateTransitionNode_15
   }
  ]
 }
 Conduit {path=AnimGraph.AnimGraphNode_StateMachine_7.LocomotionStateMachine.AnimStateConduitNode_0.Conduit, nodes [Result(bCanEnterTransition=true)]}
 EmoteLoops {
  path=AnimGraph.AnimGraphNode_StateMachine_7.LocomotionStateMachine.AnimStateNode_0.EmoteLoops
  nodes [
   AnimGraphNode_ChooserPlayer_1 = Chooser Player: PRA_Companion_SpecialIdle() {
    chooser {
     _structType=&LookupProxy
     proxy=&PRA_Companion_SpecialIdle
     proxyTable {
      _structType=&ProxyTableContextProperty
      binding {contextIndex=0, displayName=Chooser Tables, isBoundToRoot=false, propertyBindingChain [ChooserTables]}
     }
    }
    chooserContextDefinition [
     {
      _structType=&ContextObjectTypeClass
      class=&Biped_Raptor_Base_AnimBP_C
      direction=ReadWrite
     }
     {struct=&ChooserPlayerSettings, _structType=&ContextObjectTypeStruct, direction=Read}
    ]
   }
   Output Animation Pose(Result=@AnimGraphNode_ChooserPlayer_1)
  ]
 }
 AnimationBlendStackGraph_0 {
  path=AnimGraph.AnimGraphNode_StateMachine_7.LocomotionStateMachine.AnimStateNode_0.EmoteLoops.AnimGraphNode_ChooserPlayer_1.AnimationBlendStackGraph_0
  nodes [Output Pose(Result=Blend Stack Input())]
 }
 TransitionToDefaultIdle {
  path=AnimGraph.AnimGraphNode_StateMachine_7.LocomotionStateMachine.AnimStateNode_1.TransitionToDefaultIdle
  nodes [
   AnimGraphNode_ChooserPlayer_1 = Chooser Player: PRA_Companion_ReactToIdle() {
    chooser {
     _structType=&LookupProxy
     proxy=&PRA_Companion_ReactToIdle
     proxyTable {
      _structType=&ProxyTableContextProperty
      binding {contextIndex=0, displayName=Chooser Tables, isBoundToRoot=false, propertyBindingChain [ChooserTables]}
     }
    }
    chooserContextDefinition [
     {
      _structType=&ContextObjectTypeClass
      class=&Biped_Raptor_Base_AnimBP_C
      direction=ReadWrite
     }
     {struct=&ChooserPlayerSettings, _structType=&ContextObjectTypeStruct, direction=Read}
    ]
   }
   Output Animation Pose(Result=@AnimGraphNode_ChooserPlayer_1) {
    node {stateEntryFunction {functionName=OnStateEntry_0}, stateExitFunction {functionName=OnStateExit_0}}
    stateEntryFunction {bSelfContext=true, memberGuid=#27, memberName=OnStateEntry_0}
    stateExitFunction {bSelfContext=true, memberGuid=#28, memberName=OnStateExit_0}
   }
  ]
 }
 AnimationBlendStackGraph_0 {
  path=AnimGraph.AnimGraphNode_StateMachine_7.LocomotionStateMachine.AnimStateNode_1.TransitionToDefaultIdle.AnimGraphNode_ChooserPlayer_1.AnimationBlendStackGraph_0
  nodes [Output Pose(Result=Blend Stack Input())]
 }
 LocoStop {
  path=AnimGraph.AnimGraphNode_StateMachine_7.LocomotionStateMachine.AnimStateNode_2.LocoStop
  nodes [
   AnimGraphNode_ChooserPlayer_2 = Chooser Player: PRA_Companion_Stops!(
    DefaultSettings_bMirror:bool=False
    DefaultSettings_StartTime:float=Get StartStopPosition() -> (StartStopPosition:float)
    DefaultSettings_bForceLooping:bool=False
    DefaultSettings_PlaybackRate:float=1
    DefaultSettings_CurveOverrides:Anim Curve Override List Structure=(Values=,Hash=0)
    DefaultSettings_BlendTime:float=0.2
    DefaultSettings_BlendProfile:Blend Profile Object Reference
    DefaultSettings_BlendOption:EAlphaBlendOption Enum=HermiteCubic
    DefaultSettings_bUseInertialBlend:bool=False
    DefaultSettings_bForceBlendTo:bool=False
    DefaultSettings_MinDeltaTimeToForceBlendTo:float=0.1
   ) -> (Pose:Pose Link Structure) {
    is=title AnimGraphNode_ChooserPlayer
    chooser {
     _structType=&LookupProxy
     proxy=/CosmeticCompanions/Chooser/PRA_Companion_Stops
     proxyTable {
      _structType=&ProxyTableContextProperty
      binding {contextIndex=0, displayName=Chooser Tables, isBoundToRoot=false, propertyBindingChain [ChooserTables]}
     }
    }
    chooserContextDefinition [
     {
      _structType=&ContextObjectTypeClass
      class=&Biped_Raptor_Base_AnimBP_C
      direction=ReadWrite
     }
     {struct=&ChooserPlayerSettings, _structType=&ContextObjectTypeStruct, direction=Read}
    ]
   }
   Output Animation Pose(Result=@AnimGraphNode_ChooserPlayer_2)
  ]
 }
 AnimationBlendStackGraph_0 {
  path=AnimGraph.AnimGraphNode_StateMachine_7.LocomotionStateMachine.AnimStateNode_2.LocoStop.AnimGraphNode_ChooserPlayer_2.AnimationBlendStackGraph_0
  nodes [Output Pose(Result=Blend Stack Input())]
 }
 Locomotion {
  path=AnimGraph.AnimGraphNode_StateMachine_7.LocomotionStateMachine.AnimStateNode_4.Locomotion
  nodes [Output Animation Pose(Result=Use cached pose 'LocomotionPose'())]
 }
 Idle {
  path=AnimGraph.AnimGraphNode_StateMachine_7.LocomotionStateMachine.AnimStateNode_5.Idle
  nodes [Output Animation Pose(Result=Use cached pose 'IdlePose'())]
 }
 AnimationTransitionGraph_0 {path=AnimGraph.AnimGraphNode_StateMachine_7.LocomotionStateMachine.AnimStateTransitionNode_0.AnimationTransitionGraph_0, nodes [Result()]}
 AnimationTransitionGraph_0 {
  path=AnimGraph.AnimGraphNode_StateMachine_7.LocomotionStateMachine.AnimStateTransitionNode_1.AnimationTransitionGraph_0
  nodes [Result(bCanEnterTransition=true)]
 }
 AnimationTransitionGraph_0 {
  path=AnimGraph.AnimGraphNode_StateMachine_7.LocomotionStateMachine.AnimStateTransitionNode_10.AnimationTransitionGraph_0
  nodes [Result(bCanEnterTransition=Get bEnterEmoteLoops())]
 }
 AnimationTransitionGraph_0 {
  path=AnimGraph.AnimGraphNode_StateMachine_7.LocomotionStateMachine.AnimStateTransitionNode_11.AnimationTransitionGraph_0
  nodes [Result(bCanEnterTransition=NOT Boolean(A=Get bIsPlayingEmote()))]
 }
 AnimationTransitionGraph_0 {
  path=AnimGraph.AnimGraphNode_StateMachine_7.LocomotionStateMachine.AnimStateTransitionNode_12.AnimationTransitionGraph_0
  nodes [Result(bCanEnterTransition=Get bIsReactionHappyPlaying())]
 }
 AnimationTransitionGraph_0 {
  path=AnimGraph.AnimGraphNode_StateMachine_7.LocomotionStateMachine.AnimStateTransitionNode_13.AnimationTransitionGraph_0
  nodes [Result(bCanEnterTransition=Get bIsMoving())]
 }
 AnimationTransitionGraph_0 {
  path=AnimGraph.AnimGraphNode_StateMachine_7.LocomotionStateMachine.AnimStateTransitionNode_14.AnimationTransitionGraph_0
  nodes [Result(bCanEnterTransition=Get bIsMoving())]
 }
 AnimationTransitionGraph_0 {path=AnimGraph.AnimGraphNode_StateMachine_7.LocomotionStateMachine.AnimStateTransitionNode_15.AnimationTransitionGraph_0, nodes [Result()]}
 AnimationTransitionGraph_0 {
  path=AnimGraph.AnimGraphNode_StateMachine_7.LocomotionStateMachine.AnimStateTransitionNode_16.AnimationTransitionGraph_0
  nodes [Result(bCanEnterTransition=Get bIsMoving())]
 }
 AnimationTransitionGraph_0 {
  path=AnimGraph.AnimGraphNode_StateMachine_7.LocomotionStateMachine.AnimStateTransitionNode_17.AnimationTransitionGraph_0
  nodes [Result(bCanEnterTransition=Get bIsMoving())]
 }
 AnimationTransitionGraph_0 {
  path=AnimGraph.AnimGraphNode_StateMachine_7.LocomotionStateMachine.AnimStateTransitionNode_2.AnimationTransitionGraph_0
  nodes [Result(bCanEnterTransition=true)]
 }
 Transition {
  path=AnimGraph.AnimGraphNode_StateMachine_7.LocomotionStateMachine.AnimStateTransitionNode_3.Transition
  nodes [Result(bCanEnterTransition=Get bIsMoodCombat())]
 }
 AnimationTransitionGraph_0 {
  path=AnimGraph.AnimGraphNode_StateMachine_7.LocomotionStateMachine.AnimStateTransitionNode_4.AnimationTransitionGraph_0
  nodes [Result(bCanEnterTransition=NOT Boolean(A=Get bIsMoving()))]
 }
 EventGraph {nodes [
   Event BlueprintInitializeAnimation()
   n1 = Get Data!(self:Fort User Defined Data Asset Object Reference=GetAnimDataAsset!(self:Fort Player Companion Anim Instance Object Reference) -> (ReturnValue:Fort User Defined Data Asset Object Reference)) -> (bSucceed:bool, Dest:Biped Raptor Base Struct Structure) {is=title K2Node_CallFunction}
   > n2 = Branch(Condition=@n1)
   > Set Biped_Raptor_Base_Struct(Biped_Raptor_Base_Struct:Biped Raptor Base Struct Structure=@n1.Dest) -> (Output_Get:Biped Raptor Base Struct Structure)
   n3 = Get Data(self=Get Biped_Raptor_DA() -> (Biped_Raptor_DA:Fort User Defined Data Asset Object Reference))
   Branch(execute=@n2.else, Condition=@n3)
   > Set Biped_Raptor_Base_Struct(Biped_Raptor_Base_Struct=@n3.Dest)
  ]}
 GetSequencerAnimSlotNames {nodes [
   GetSequencerAnimSlotNames() {is=title K2Node_FunctionEntry}
   > Return Node(ReturnValue:Array of Names=Make Array("[0]":Name=SequencerDefault) -> (Array:Array of Names))
  ]}
 OnStateEntry_0 {nodes [
   OnStateEntry_0!() -> (then, Context:"Anim Update Context Structure (by ref)", Node:"Anim Node Reference Structure (by ref)") {is=title K2Node_FunctionEntry}
   > Set bIsSpecialIdleFinish(bIsSpecialIdleFinish:bool=true) -> (Output_Get:bool)
   > Return Node!(execute)
  ]}
 OnStateExit_0 {nodes [
   OnStateExit_0!() -> (then, Context:"Anim Update Context Structure (by ref)", Node:"Anim Node Reference Structure (by ref)") {is=title K2Node_FunctionEntry}
   > Set bIsSpecialIdleFinish(bIsSpecialIdleFinish=false)
   > Return Node()
  ]}
 ToSitting_Entry {nodes [
   ToSitting_Entry!() -> (then, Context:"Anim Update Context Structure (by ref)", Node:"Anim Node Reference Structure (by ref)") {is=title K2Node_FunctionEntry}
   > Set bIsSitting(bIsSitting:bool=true) -> (Output_Get:bool)
   > Return Node()
  ]}
 ToSitting_Exit {nodes [
   ToSitting_Exit!() -> (then, Context:"Anim Update Context Structure (by ref)", Node:"Anim Node Reference Structure (by ref)") {is=title K2Node_FunctionEntry}
   > Set bIsSitting(bIsSitting=false)
   > Return Node()
  ]}
]
variables [
 Biped_Raptor_Base_Struct {
  type=Biped_Raptor_Base_Struct
  default {
   aimOffset=null
   angry_Idle_Intro=null
   angry_Idle_Loop=null
   angry_Idle_Outro=null
   bodyBend_BS=null
   bodyPose_Angry=null
   bodyPose_Worried=null
   featherSpreadPose=null
   locomotion_Idle=null
   locomotion_Jog=null
   locomotion_Run=null
   locomotion_Sprint=null
   locomotion_Start=null
   locomotion_Walk=null
   sitToStand=null
   sitting=null
   standToSit=null
   tailBend_BS=null
  }
 }
 Biped_Raptor_DA {type=FortUserDefinedDataAsset, default=null}
 ChooserTables {type=ProxyTable, default=null}
]
